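# Base volume material for Niagara Fluids 3D gas, rendering smoke and fire through heterogeneous volumes.
# World position maps into grid space to sample volume textures, with density from red, temperature from blue and optional color.
# Remapped density drives extinction, while temperature drives fire emission via blackbody or ramp, with optional opacity curve and density-based scatter masking.
# Static switches select these variants, including Color To Emissive, which zeroes albedo and emits smoke color instead.
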Material /NiagaraFluids/Materials/Grid3D/M_3DGas_Base
graph {
 properties {
  bUsedWithHeterogeneousVolumes=true
  bUsedWithNiagaraMeshParticles=true
  blendMode=BLEND_Additive
  materialDomain=MD_Volume
  pixelDepthOffsetMode=PDOM_Legacy
  usedShadingModels=Default Lit
 }
 nodes [
  n1 = StaticBoolParameter() {parameterName=Color To Emissive}
  n2 = WorldPosition()
  n3 = MaterialFunctionCall(
   VectorToTransform=@n2.XYZ
   BasisX=VectorParameter() {parameterName=WorldToLocal0}.RGB
   BasisY=VectorParameter() {parameterName=WorldToLocal1}.RGB
   BasisZ=VectorParameter() {parameterName=WorldToLocal2}.RGB
   BasisPosition=VectorParameter() {parameterName=WorldToLocal3}.RGB
  ) {materialFunction=&Transform3x3Matrix}
  n4 = VectorParameter() {parameterName=WorldGridExtents}
  n5 = Divide(A=@n4.RGB)
  n6 = TransformPosition(None=@n2.XYZ) {transformSourceType=TRANSFORMPOSSOURCE_World}
  n7 = VectorParameter() {parameterName=WorldGridExtents}
  n8 = StaticSwitchParameter(
   True=Divide(A=Add(A=@n3.Transformed Vector, B=@n5), B=@n4.RGB)
   False=Add(A=Divide(A=@n6, B=@n7.RGB), B=VectorParameter() {parameterName=LocalPivot}.RGB)
  ) {defaultValue=true, parameterName=Use Legacy Transform}
  n9 = TextureSampleParameterVolume!(UVs=@n8, Apply View MipBias) {parameterName=ColorTexture}
  n10 = Multiply(
   A=StaticSwitchParameter(True=@n9.RGB, False=VectorParameter() {parameterName=Smoke Color, defaultValue=<rgba 1 1 1 0>}.RGB) {parameterName=Use Colored Smoke}
   B=ScalarParameter() {defaultValue=1, parameterName=Albedo}
  )
  StaticSwitch(True=Constant3Vector(), False=@n10, Value=@n1)
  n11 = ScalarParameter() {parameterName=DensityOffset}
  n12 = TextureSampleParameterVolume(UVs=@n8) {parameterName=VolumeTex}
  n13 = StaticSwitchParameter(True=@n12.R, False=@n12.R) {defaultValue=true, parameterName=UseLegacyVolumeSample}
  n14 = ScalarParameter() {parameterName=DensityRange}
  n15 = Constant() {r=1}
  n16 = Saturate(None=Divide(A=Subtract(A=@n13, B=@n11), B=If(A=@n14, "A > B"=@n14, "A == B"=@n15, "A < B"=@n15)))
  n17 = Multiply(
   A=StaticSwitchParameter(
    True=TextureSampleParameter2D(UVs=@n16) {parameterName=DensityRamp, samplerType=SAMPLERTYPE_LinearColor, texture=/Engine/ArtTools/RenderToTexture/Textures/127grey}.R
    False=@n16
   ) {parameterName=Use Density Curve}
   B=ScalarParameter() {defaultValue=1, parameterName=DensityGain}
  )
  n18 = If(A=@n17, "A > B"=Add(A=@n17, B=ScalarParameter() {parameterName=ConstantExtinction}), "A == B"=@n17, "A < B"=@n17) {constB=0}
  n19 = StaticSwitchParameter(True=@n9.RGB, False=VectorParameter() {parameterName=Fire Color, defaultValue=<rgba 1 1 1 0>}.RGB) {parameterName=Use Colored Fire}
  n20 = Constant() {r=4500}
  n21 = StaticSwitchParameter(True=@n12.B, False=@n12.B) {defaultValue=true, parameterName=UseLegacyVolumeSample}
  n22 = ScalarParameter() {parameterName=TemperatureOffset}
  n23 = ScalarParameter() {parameterName=TemperatureRange}
  n24 = Constant() {r=1}
  n25 = Saturate(None=Divide(A=Subtract(A=@n21, B=@n22), B=If(A=@n23, "A > B"=@n23, "A == B"=@n24, "A < B"=@n24)))
  n26 = StaticSwitchParameter(
   True=BlackBody!(Temp=Multiply(A=@n20, B=@n25))
   False=TextureSampleParameter2D(UVs=@n25) {parameterName=TemperatureRamp, samplerType=SAMPLERTYPE_LinearColor, texture=&127grey}.RGB
  ) {parameterName=Use Blackbody Curve}
  n27 = TextureSampleParameter2D(UVs=@n25) {parameterName=TemperatureOpacityRamp, samplerType=SAMPLERTYPE_LinearColor, texture=&127grey}
  n28 = Multiply(
   A=Multiply(A=Multiply(A=@n19, B=@n26), B=ScalarParameter() {defaultValue=1, parameterName=FireGain})
   B=Multiply(
    A=StaticSwitchParameter(True=@n27.R, False=@n25) {parameterName=Use Temperature Opacity Curve}
    B=ScalarParameter() {defaultValue=1, parameterName=FireOpacityGain}
   )
  )
  n29 = ScalarParameter() {defaultValue=0.9, parameterName=ScatterMaskContribution}
  n30 = ScalarParameter() {defaultValue=0.3, parameterName=ScatterMaskPosition}
  n31 = Multiply(A=ScalarParameter() {defaultValue=0.3, parameterName=ScatterMaskWidth}) {constB=0.5}
  n32 = Saturate(None=MaterialFunctionCall!(Input=@n13, Input Low=Subtract(A=@n30, B=@n31), Input High=Add(A=@n30, B=@n31)) {materialFunction=/Engine/Functions/Engine_MaterialFunctions03/Math/RemapValueRangeNormalized}.Result)
  n33 = Multiply(A=@n13, B=@n28)
  n34 = Multiply(A=LinearInterpolate(A=OneMinus(None=@n29), Alpha=OneMinus(None=@n32)), B=@n33)
  StaticSwitch(
   True=StaticSwitchParameter(True=Multiply(A=@n10, B=@n18), False=@n10) {parameterName=Multiply Color With Extinction}
   False=StaticSwitchParameter(True=StaticSwitchParameter(True=@n34, False=@n33) {parameterName=Use Scatter Mask}, False=@n28) {parameterName=Use Scatter}
   Value=@n1
  )
 ]
}
