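# Material function that computes world position offset for a Niagara Fluids liquid volume, moving the box's vertices onto screen corners just past the camera.
# Custom HLSL rounds each vertex's box-space XY to a corner and unprojects it through ScreenToWorld at a fixed depth.
# Static switch "Use Legacy Transform" derives box space from matrix parameters or from TransformPosition plus LocalPivot.
# Orthographic views get zero offset; an unconnected Custom node holds a bounding-box clipping variant.
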MaterialFunction /NiagaraFluids/Materials/Functions/Niagara_Liquid_Generate_WPO
graph {nodes [
  Custom!(
   InWorld
   CameraWorld
   WorldGridExtents
   WorldToLocal0
   WorldToLocal1
   WorldToLocal2
   WorldToLocal3
   LocalToWorld0
   LocalToWorld1
   LocalToWorld2
   LocalToWorld3
   NearClipOffset
  ) {
   code="float4x4 WorldToLocal = {\r\nWorldToLocal0.x, WorldToLocal0.y, WorldToLocal0.z, WorldToLocal0.w, \r\nWorldToLocal1.x, WorldToLocal1.y, WorldToLocal1.z, WorldToLocal1.w, \r\nWorldToLocal2.x, WorldToLocal2.y, WorldToLocal2.z, WorldToLocal2.w, \r\nWorldToLocal3.x, WorldToLocal3.y, WorldToLocal3.z, WorldToLocal3.w\r\n};\r\n\r\nfloat4x4 LocalToWorld = {\r\nLocalToWorld0.x, LocalToWorld0.y, LocalToWorld0.z, LocalToWorld0.w, \r\nLocalToWorld1.x, LocalToWorld1.y, LocalToWorld1.z, LocalToWorld1.w, \r\nLocalToWorld2.x, LocalToWorld2.y, LocalToWorld2.z, LocalToWorld2.w, \r\nLocalToWorld3.x, LocalToWorld3.y, LocalToWorld3.z, LocalToWorld3.w\r\n};\r\n\r\n// Unit space position of the current vertex\r\nfloat3 InUnit = mul(float4(InWorld,1), WorldToLocal).xyz;\r\nInUnit = InUnit / WorldGridExtents + .5;\r\n\r\n\r\nfloat3 CameraUnit = mul(float4(CameraWorld,1), WorldToLocal).xyz;\r\nCameraUnit = CameraUnit / WorldGridExtents + .5;\r\n\r\nbool IsInside = all(CameraUnit <= float3(1,1,1)+1e-5) &&&& all(CameraUnit >= float3(0,0,0)-1e-5);\r\n\r\nfloat3 OutWorldVertexPos = InWorld;\r\n// true if at least one vertex is behind the near clip plane\r\nbool VertexBehind = false;\r\n\r\n// true if at least one vertex is visible\r\nbool VertexVisible = false;\r\n\r\n// evaluate the 8 corners of the fluid bbox in clip space \r\n// find the axis aligned bbox in clip space\r\nfloat4 BBoxClipMin = float4(INFINITE_FLOAT,INFINITE_FLOAT,INFINITE_FLOAT,INFINITE_FLOAT);\r\nfloat4 BBoxClipMax = -1. * float4(INFINITE_FLOAT,INFINITE_FLOAT,INFINITE_FLOAT,INFINITE_FLOAT);\r\n\r\nfor (int x = 0; x <= 1; ++x) {\r\nfor (int y = 0; y <= 1; ++y) {\r\nfor (int z = 0; z <= 1; ++z) {\r\n\r\n    // get the fluid box corner in world space\r\n    const float3 BBoxUnit = float3(x,y,z);\r\n    const float3 BBoxLocal = (BBoxUnit - .5) * WorldGridExtents;\r\n    const float3 BBoxWorld =  mul(float4(BBoxLocal,1), LocalToWorld).xyz;\r\n\r\n\r\n    // world to clip space\r\n    // assume no LWC for now\r\n    FDFVector4 FDFBBoxWorld = \r\n        DFFromTileOffset_Hack(MakeLWCVector4(MakeLWCVector3(float3(0,0,0), BBoxWorld), 1));\r\n    float4 BBoxClip = DFMultiplyDemote(FDFBBoxWorld, ResolvedView.WorldToClip);\r\n    float2 ScreenPos = BBoxClip.xy / BBoxClip.w;\r\n    // if this vertex is in front of the near clip plane, add it to the aabbox\r\n    if (BBoxClip.w > 0)    \r\n    {\r\n        // vertex is visible\r\n        if (any(abs(ScreenPos) <= float2(1,1)))\r\n        {\r\n            VertexVisible = true;\r\n        }\r\n   \r\n        BBoxClipMin = min(BBoxClip, BBoxClipMin);\r\n        BBoxClipMax = max(BBoxClip, BBoxClipMax);      \r\n    } \r\n    else\r\n    {\r\n       VertexBehind = true;\r\n    }\r\n}}}\r\nBBoxClipMin = min(BBoxClipMin, BBoxClipMax);\r\nBBoxClipMax = max(BBoxClipMin, BBoxClipMax);\r\n// if we have a visible vertex and at least one vertex is behind the near clip plane\r\n// then transform to a clip aligned bbox\r\n\r\nif (true)\r\n{\r\n    int3 UnitCorner = round(InUnit);\r\n    UnitCorner.z = 0;\r\n    float2 ScreenPosition = 2 * (round(InUnit.xy) - .5);\r\n    float SceneDepth = NearClipOffset;//max(NearClipOffset, BBoxClipMin.z / BBoxClipMin.w);\r\n    float4 Tmp = mul(float4(ScreenPosition * SceneDepth, SceneDepth, 1), DFHackToFloat(ResolvedView.ScreenToWorld));\r\n    OutWorldVertexPos = Tmp.xyz / Tmp.w;\r\n}\r\nelse if (VertexVisible &&&& VertexBehind)\r\n{\r\n\r\n      float4 BBoxClipSize = BBoxClipMax - BBoxClipMin;\r\n\r\n      // transform the box back to world space for the current vertex.  \r\n      // vertex is identified by the unit space position\r\n      const int3 UnitCorner = round(InUnit);\r\n      int4 v = float4(1.0 - UnitCorner, UnitCorner.z);\r\n      float4 BBoxClipPos = BBoxClipMin + BBoxClipSize * v;\r\n \r\n      // clamp to a near clip plane offset\r\n     // BBoxClipPos.w = max(NearClipOffset, BBoxClipPos.w);\r\n\r\n      // output world position for vertex\r\n      float4 Tmp = DFDemote(DFMultiply(BBoxClipPos, ResolvedView.ClipToWorld));\r\n      OutWorldVertexPos = Tmp.xyz;\r\n}\r\n\r\nreturn OutWorldVertexPos;"
   showCode=true
   inputs [
    {inputName=InWorld, input {expression=null}}
    {inputName=CameraWorld, input {expression=null}}
    {inputName=WorldGridExtents, input {expression=null, mask=1, maskB=1, maskG=1, maskR=1}}
    {inputName=WorldToLocal0, input {expression=null}}
    {inputName=WorldToLocal1, input {expression=null}}
    {inputName=WorldToLocal2, input {expression=null}}
    {inputName=WorldToLocal3, input {expression=null}}
    {inputName=LocalToWorld0, input {expression=null}}
    {inputName=LocalToWorld1, input {expression=null}}
    {inputName=LocalToWorld2, input {expression=null}}
    {inputName=LocalToWorld3, input {expression=null}}
    {inputName=NearClipOffset, input {expression=null}}
   ]
  }
  n1 = Constant3Vector() {desc=Disable In Ortho Views for now}
  MaterialExpressionWorldPosition_1 = WorldPosition() {worldPositionShaderOffset=WPT_ExcludeAllShaderOffsets}
  MaterialExpressionCameraPositionWS_0 = CameraPositionWS()
  MaterialExpressionVectorParameter_0 = VectorParameter() {parameterName=WorldGridExtents}
  n2 = VectorParameter() {parameterName=WorldToLocal0}
  MaterialExpressionAppendVector_1 = AppendVector(A=@n2.RGB, B=@n2.RGB)
  n3 = VectorParameter() {parameterName=WorldToLocal1}
  MaterialExpressionAppendVector_2 = AppendVector(A=@n3.RGB, B=@n3.RGB)
  n4 = VectorParameter() {parameterName=WorldToLocal2}
  MaterialExpressionAppendVector_3 = AppendVector(A=@n4.RGB, B=@n4.RGB)
  n5 = VectorParameter() {parameterName=WorldToLocal3}
  MaterialExpressionAppendVector_4 = AppendVector(A=@n5.RGB, B=@n5.RGB)
  n6 = VectorParameter() {parameterName=LocalToWorld0}
  MaterialExpressionAppendVector_0 = AppendVector(A=@n6.RGB, B=@n6.RGB)
  n7 = VectorParameter() {parameterName=LocalToWorld1}
  MaterialExpressionAppendVector_5 = AppendVector(A=@n7.RGB, B=@n7.RGB)
  n8 = VectorParameter() {parameterName=LocalToWorld2}
  MaterialExpressionAppendVector_6 = AppendVector(A=@n8.RGB, B=@n8.RGB)
  n9 = VectorParameter() {parameterName=LocalToWorld3}
  MaterialExpressionAppendVector_7 = AppendVector(A=@n9.RGB, B=@n9.RGB)
  MaterialExpressionConstant_0 = Constant() {r=15}
  n10 = Custom(
   InWorld=@MaterialExpressionWorldPosition_1.XYZ
   CameraWorld=@MaterialExpressionCameraPositionWS_0
   WorldGridExtents=@MaterialExpressionVectorParameter_0.RGB
   WorldToLocal0=@MaterialExpressionAppendVector_1
   WorldToLocal1=@MaterialExpressionAppendVector_2
   WorldToLocal2=@MaterialExpressionAppendVector_3
   WorldToLocal3=@MaterialExpressionAppendVector_4
   LocalToWorld0=@MaterialExpressionAppendVector_0
   LocalToWorld1=@MaterialExpressionAppendVector_5
   LocalToWorld2=@MaterialExpressionAppendVector_6
   LocalToWorld3=@MaterialExpressionAppendVector_7
   NearClipOffset=@MaterialExpressionConstant_0
  ) {
   code="float4x4 WorldToLocal = {\r\nWorldToLocal0.x, WorldToLocal0.y, WorldToLocal0.z, WorldToLocal0.w, \r\nWorldToLocal1.x, WorldToLocal1.y, WorldToLocal1.z, WorldToLocal1.w, \r\nWorldToLocal2.x, WorldToLocal2.y, WorldToLocal2.z, WorldToLocal2.w, \r\nWorldToLocal3.x, WorldToLocal3.y, WorldToLocal3.z, WorldToLocal3.w\r\n};\r\n\r\nfloat4x4 LocalToWorld = {\r\nLocalToWorld0.x, LocalToWorld0.y, LocalToWorld0.z, LocalToWorld0.w, \r\nLocalToWorld1.x, LocalToWorld1.y, LocalToWorld1.z, LocalToWorld1.w, \r\nLocalToWorld2.x, LocalToWorld2.y, LocalToWorld2.z, LocalToWorld2.w, \r\nLocalToWorld3.x, LocalToWorld3.y, LocalToWorld3.z, LocalToWorld3.w\r\n};\r\n\r\n// Unit space position of the current vertex\r\nfloat3 InUnit = mul(float4(InWorld,1), WorldToLocal).xyz;\r\nInUnit = InUnit / WorldGridExtents + .5;\r\n\r\n\r\nfloat3 CameraUnit = mul(float4(CameraWorld,1), WorldToLocal).xyz;\r\nCameraUnit = CameraUnit / WorldGridExtents + .5;\r\n\r\nint3 UnitCorner = round(InUnit);\r\nUnitCorner.z = 0;\r\nfloat2 ScreenPosition = 2 * (round(InUnit.xy) - .5);\r\nfloat SceneDepth = NearClipOffset;//max(NearClipOffset, BBoxClipMin.z / BBoxClipMin.w);\r\nfloat4 Tmp = mul(float4(ScreenPosition * SceneDepth, SceneDepth, 1), DFHackToFloat(ResolvedView.ScreenToWorld));\r\n\r\n\r\nreturn Tmp.xyz / Tmp.w;"
   showCode=true
   inputs [
    {inputName=InWorld, input {expression=&self:MaterialExpressionWorldPosition_1}}
    {
     inputName=CameraWorld
     input {expression=&self:MaterialExpressionCameraPositionWS_0}
    }
    {
     inputName=WorldGridExtents
     input {
      expression=&self:MaterialExpressionVectorParameter_0
      mask=1
      maskB=1
      maskG=1
      maskR=1
     }
    }
    {inputName=WorldToLocal0, input {expression=&self:MaterialExpressionAppendVector_1}}
    {inputName=WorldToLocal1, input {expression=&self:MaterialExpressionAppendVector_2}}
    {inputName=WorldToLocal2, input {expression=&self:MaterialExpressionAppendVector_3}}
    {inputName=WorldToLocal3, input {expression=&self:MaterialExpressionAppendVector_4}}
    {inputName=LocalToWorld0, input {expression=&self:MaterialExpressionAppendVector_0}}
    {inputName=LocalToWorld1, input {expression=&self:MaterialExpressionAppendVector_5}}
    {inputName=LocalToWorld2, input {expression=&self:MaterialExpressionAppendVector_6}}
    {inputName=LocalToWorld3, input {expression=&self:MaterialExpressionAppendVector_7}}
    {inputName=NearClipOffset, input {expression=&self:MaterialExpressionConstant_0}}
   ]
  }
  MaterialExpressionWorldPosition_0 = WorldPosition() {worldPositionShaderOffset=WPT_ExcludeAllShaderOffsets}
  MaterialExpressionCameraPositionWS_1 = CameraPositionWS()
  MaterialExpressionVectorParameter_1 = VectorParameter() {parameterName=WorldGridExtents}
  MaterialExpressionTransformPosition_0 = TransformPosition(None=@MaterialExpressionWorldPosition_0.XYZ) {transformSourceType=TRANSFORMPOSSOURCE_World}
  MaterialExpressionTransformPosition_1 = TransformPosition(None=@MaterialExpressionCameraPositionWS_1) {transformSourceType=TRANSFORMPOSSOURCE_World}
  MaterialExpressionAdd_0 = Add(A=VectorParameter() {parameterName=LocalPivot}.RGB) {constB=0}
  n11 = Custom!(
   InWorld=@MaterialExpressionWorldPosition_0.XYZ
   CameraWorld=@MaterialExpressionCameraPositionWS_1
   WorldGridExtents=@MaterialExpressionVectorParameter_1.RGB
   NearClipOffset=@MaterialExpressionConstant_0
   InLocal=@MaterialExpressionTransformPosition_0
   CameraLocal=@MaterialExpressionTransformPosition_1
   LocalPivot=@MaterialExpressionAdd_0
  ) {
   code="// Unit space position of the current vertex\r\nfloat3 InUnit = InLocal / WorldGridExtents + LocalPivot;\r\nfloat3 CameraUnit = CameraLocal / WorldGridExtents + LocalPivot;\r\n\r\nint3 UnitCorner = round(InUnit);\r\nUnitCorner.z = 0;\r\nfloat2 ScreenPosition = 2 * (round(InUnit.xy) - .5);\r\nfloat SceneDepth = NearClipOffset;//max(NearClipOffset, BBoxClipMin.z / BBoxClipMin.w);\r\nfloat4 Tmp = mul(float4(ScreenPosition * SceneDepth, SceneDepth, 1), DFHackToFloat(ResolvedView.ScreenToWorld));\r\n\r\nreturn Tmp.xyz / Tmp.w;"
   showCode=true
   inputs [
    {
     inputName=InWorld
     input {
      expression=&self:MaterialExpressionWorldPosition_0
      mask=1
      maskB=1
      maskG=1
      maskR=1
     }
    }
    {
     inputName=CameraWorld
     input {expression=&self:MaterialExpressionCameraPositionWS_1}
    }
    {
     inputName=WorldGridExtents
     input {
      expression=&self:MaterialExpressionVectorParameter_1
      mask=1
      maskB=1
      maskG=1
      maskR=1
     }
    }
    {inputName=NearClipOffset, input {expression=&self:MaterialExpressionConstant_0}}
    {inputName=InLocal, input {expression=&self:MaterialExpressionTransformPosition_0}}
    {
     inputName=CameraLocal
     input {expression=&self:MaterialExpressionTransformPosition_1}
    }
    {inputName=LocalPivot, input {expression=&self:MaterialExpressionAdd_0}}
   ]
  }
  n12 = StaticSwitchParameter(True=@n10, False=@n11) {defaultValue=true, parameterName=Use Legacy Transform}
  FunctionOutput(None=If(A=IsOrthographic(), "A > B"=@n1, "A == B"=@n1, "A < B"=Subtract(A=@n12, B=@MaterialExpressionWorldPosition_1.XYZ)) {constB=1, desc=Ray Start Ortho Switch})
 ]}
