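# Unlit debug material for Niagara Fluids Grid3D, usable on mesh particles, that visualizes a depth texture as a gradient.
# It samples the DepthTex parameter at viewport UVs, divides red by a constant and takes the fraction, producing repeating depth bands.
# MF_DebugGradient colors the bands, and EyeAdaptationInverse compensates the result for auto exposure.
# The graph also includes the Niagara_Liquid_Generate_WPO function for world position offset.
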Material /NiagaraFluids/Materials/Grid3D/M_DebugDepth
graph {
 properties {bUsedWithNiagaraMeshParticles=true, pixelDepthOffsetMode=PDOM_Legacy, shadingModel=MSM_Unlit, usedShadingModels=Unlit}
 nodes [
  n1 = ScreenPosition()
  n2 = Frac(None=Divide(A=TextureSampleParameter2D(UVs=@n1.ViewportUV) {parameterName=DepthTex, texture=&Black}.R) {constB=1000})
  EyeAdaptationInverse(LightValueInput=ComponentMask(None=MaterialFunctionCall!(X=@n2) {materialFunction=/NiagaraFluids/Materials/Functions/MF_DebugGradient}.Result) {b=true, g=true, r=true})
  MaterialFunctionCall() {materialFunction=/NiagaraFluids/Materials/Functions/Niagara_Liquid_Generate_WPO}
 ]
}
