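# Translucent Niagara debug material that renders a signed distance field from a 3D grid volume texture as a surface.
# A Custom node moves the camera ray into grid space, intersects the bounds and sphere-traces the red channel, hiding hits behind scene depth.
# Normals come from central differences or stored GBA, and feed a fixed-direction dot product for shading.
# Debug Bounds draws a voxel checkerboard instead; the slice center and axis inputs are unused.
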Material /NiagaraFluids/Materials/Grid3D/M_DebugSDF
graph {
 properties {
  bDisableDepthTest=true
  bUsedWithNiagaraMeshParticles=true
  bUsedWithNiagaraSprites=true
  blendMode=BLEND_Translucent
  pixelDepthOffsetMode=PDOM_Legacy
  usedShadingModels=Default Lit
 }
 nodes [
  n1 = CameraPositionWS()
  MaterialExpressionReroute_0 = Reroute(None=@n1)
  MaterialExpressionNormalize_0 = Normalize(VectorInput=Subtract(A=WorldPosition().XYZ, B=@n1))
  MaterialExpressionReroute_5 = Reroute(None=TextureObjectParameter() {parameterName=VolumeTex, samplerType=SAMPLERTYPE_LinearColor, texture=/Engine/EngineDebugMaterials/VolumeToRender})
  MaterialExpressionReroute_6 = Reroute(None=VectorParameter() {parameterName=WorldGridExtents}.RGB)
  MaterialExpressionReroute_7 = Reroute(None=ScalarParameter() {defaultValue=0.25, parameterName=Voxel Size})
  n2 = VectorParameter() {parameterName=WorldToLocal0}
  MaterialExpressionReroute_1 = Reroute(None=AppendVector(A=@n2.RGB, B=@n2.RGB))
  n3 = VectorParameter() {parameterName=WorldToLocal1}
  MaterialExpressionReroute_2 = Reroute(None=AppendVector(A=@n3.RGB, B=@n3.RGB))
  n4 = VectorParameter() {parameterName=WorldToLocal2}
  MaterialExpressionReroute_3 = Reroute(None=AppendVector(A=@n4.RGB, B=@n4.RGB))
  n5 = VectorParameter() {parameterName=WorldToLocal3}
  MaterialExpressionReroute_4 = Reroute(None=AppendVector(A=@n5.RGB, B=@n5.RGB))
  MaterialExpressionMaterialFunctionCall_1 = MaterialFunctionCall() {materialFunction=&CameraDirectionVector}
  MaterialExpressionSceneDepth_0 = SceneDepth()
  MaterialExpressionScalarParameter_4 = ScalarParameter() {parameterName=Debug Bounds}
  MaterialExpressionVectorParameter_1 = VectorParameter() {parameterName=DebugSliceCenter, defaultValue=<rgba 0 0 100 0>}
  MaterialExpressionVectorParameter_2 = VectorParameter() {parameterName=DebugSliceAxis, defaultValue=<rgba 0 1 0 0>}
  MaterialExpressionScalarParameter_1 = ScalarParameter() {defaultValue=1, parameterName=Tolerance}
  MaterialExpressionScalarParameter_2 = ScalarParameter() {defaultValue=300, parameterName=NumSteps}
  MaterialExpressionScalarParameter_3 = ScalarParameter() {defaultValue=1, parameterName=ComputeNormals}
  n6 = Custom!(
   RayStart=@MaterialExpressionReroute_0
   RayDir=@MaterialExpressionNormalize_0
   VolumeTexture=@MaterialExpressionReroute_5
   WorldGridExtents=@MaterialExpressionReroute_6
   VoxelSize=@MaterialExpressionReroute_7
   WorldToLocal0=@MaterialExpressionReroute_1
   WorldToLocal1=@MaterialExpressionReroute_2
   WorldToLocal2=@MaterialExpressionReroute_3
   WorldToLocal3=@MaterialExpressionReroute_4
   CameraDirectionVector=@MaterialExpressionMaterialFunctionCall_1.Result
   SceneDepth=@MaterialExpressionSceneDepth_0.R
   DebugBBox=@MaterialExpressionScalarParameter_4
   Center=@MaterialExpressionVectorParameter_1.RGB
   Normal=@MaterialExpressionVectorParameter_2.RGB
   Tolerance=@MaterialExpressionScalarParameter_1
   NumSteps=@MaterialExpressionScalarParameter_2
   ComputeNormals=@MaterialExpressionScalarParameter_3
  ) {
   code="float4 ret = {0,0,0,0};\r\n\r\nfloat4x4 WorldToLocal = {\r\nWorldToLocal0.x, WorldToLocal0.y, WorldToLocal0.z, WorldToLocal0.w, \r\nWorldToLocal1.x, WorldToLocal1.y, WorldToLocal1.z, WorldToLocal1.w, \r\nWorldToLocal2.x, WorldToLocal2.y, WorldToLocal2.z, WorldToLocal2.w, \r\nWorldToLocal3.x, WorldToLocal3.y, WorldToLocal3.z, WorldToLocal3.w\r\n};\r\n\r\nfloat WorldDepth = SceneDepth / dot(RayDir, CameraDirectionVector);\r\n\r\nRayStart = mul(float4(RayStart,1), WorldToLocal).xyz;\r\nRayDir = mul(float4(RayDir,0), WorldToLocal).xyz;\r\n\r\n//\r\n//  Ray aabbox intersection\r\n//\r\n\r\nfloat3 dirfrac = 1.0f / RayDir;\r\n\r\nfloat t = 0;\r\n\r\n\r\nfloat3 MinBounds = -1.* WorldGridExtents * .5;\r\nfloat3 MaxBounds = WorldGridExtents * .5;\r\nfloat t1 = (MinBounds.x - RayStart.x)*dirfrac.x;\r\nfloat t2 = (MaxBounds.x - RayStart.x)*dirfrac.x;\r\nfloat t3 = (MinBounds.y - RayStart.y)*dirfrac.y;\r\nfloat t4 = (MaxBounds.y - RayStart.y)*dirfrac.y;\r\nfloat t5 = (MinBounds.z - RayStart.z)*dirfrac.z;\r\nfloat t6 = (MaxBounds.z - RayStart.z)*dirfrac.z;\r\n\r\nfloat tmin = max(max(min(t1, t2), min(t3, t4)), min(t5, t6));\r\nfloat tmax = min(min(max(t1, t2), max(t3, t4)), max(t5, t6));\r\n\r\n// if tmax < 0, ray (line) is intersecting AABB, but the whole AABB is behind us\r\nif (tmax < 0)\r\n{\r\n    t = tmax;\r\n    return ret;\r\n}\r\n\r\n// if tmin > tmax, ray doesn't intersect AABB\r\nif (tmin > tmax)\r\n{\r\n    t = tmax;\r\n    return ret;\r\n}\r\n\r\nt = tmin;\r\n\r\nfloat3 CurrLocalPos = RayStart + RayDir * t;\r\n\r\nif (DebugBBox > 1e-3 &&&& t < WorldDepth) {\r\n  float4 rettmp = {1,1,1,DebugBBox};\r\n\r\n  int3 VoxelCount = floor(CurrLocalPos / VoxelSize);\r\n\r\n  int SliceParity = (VoxelCount.x) % 2;\r\n  int RowParity = (VoxelCount.y + SliceParity+1) % 2;\r\n  int CellParity = (VoxelCount.z + RowParity ) % 2;\r\n\r\n  if (CellParity == 0)\r\n  {\r\n    rettmp.y = 0;\r\n  }\r\n\r\n  return rettmp;\r\n}\r\n\r\nfloat SignedDistance = -10000;\r\nfloat3 FinalPos = float3(0,0,0);\r\nfloat4 RetVal = float4(0,0,0,0);\r\nfor (int i = 0; i < NumSteps; ++i)\r\n{\r\n   \r\n    float3 Position = CurrLocalPos / WorldGridExtents + .5;\r\n    float4 VolumeSample = VolumeTexture.SampleLevel(VolumeTextureSampler, Position, 0);\r\n    SignedDistance = VolumeSample.r;\r\n    if (abs(SignedDistance) < VoxelSize * Tolerance &&&& Position.x >= 0 &&&& Position.x <= 1 &&&& Position.y >= 0 &&&& Position.y <= 1 &&&& Position.z >= 0 &&&& Position.z <= 1)\r\n    {\r\n        FinalPos = Position;\r\n        RetVal = float4(normalize(VolumeSample.gba), 1);\r\n        break;\r\n    }\r\n    \r\n//    if (t > SceneDepth)\r\n//    {\r\n//        FinalPos = Position;\r\n//        RetVal = float4(normalize(VolumeSample.gba), abs(SignedDistance) <  VoxelSize * Tolerance);\r\n//        break;\r\n//    }\r\n\r\n    t += SignedDistance;\r\n    CurrLocalPos = RayStart + RayDir * t;\r\n}\r\n\r\nif (abs(SignedDistance) > VoxelSize * Tolerance)\r\n{\r\n//RetVal.a = 0;\r\n}\r\n\r\nif (RetVal.a != 0 &&&& ComputeNormals > 1-1e-5)\r\n{\r\nuint sx,sy,sz,l;\r\nVolumeTexture.GetDimensions(0, sx, sy, sz, l);\r\nfloat3 UnitDx = 1./float3(sx,sy,sz);\r\n\r\nfloat S_right = VolumeTexture.SampleLevel(VolumeTextureSampler, FinalPos + float3(1,0,0) * UnitDx, 0).r;\r\nfloat S_left = VolumeTexture.SampleLevel(VolumeTextureSampler, FinalPos - \r\nfloat3(1,0,0) * UnitDx, 0).r;\r\nfloat S_up  = VolumeTexture.SampleLevel(VolumeTextureSampler, FinalPos + \r\nfloat3(0,1,0) * UnitDx, 0).r;\r\nfloat S_down = VolumeTexture.SampleLevel(VolumeTextureSampler, FinalPos - float3(0,1,0) * UnitDx, 0).r;\r\nfloat S_front = VolumeTexture.SampleLevel(VolumeTextureSampler, FinalPos + float3(0,0,1) * UnitDx, 0).r;\r\nfloat S_back = VolumeTexture.SampleLevel(VolumeTextureSampler, FinalPos - float3(0,0,1) * UnitDx, 0).r;\r\n\r\nRetVal.rgb = normalize(float3(S_right - S_left, S_up - S_down, S_front - S_back));\r\n\r\n}\r\n\r\nif (t > WorldDepth)\r\n{\r\nRetVal.a = 0;\r\n}\r\n\r\nreturn RetVal;"
   outputType=CMOT_Float4
   inputs [
    {inputName=RayStart, input {expression=&self:MaterialExpressionReroute_0}}
    {inputName=RayDir, input {expression=&self:MaterialExpressionNormalize_0}}
    {inputName=VolumeTexture, input {expression=&self:MaterialExpressionReroute_5}}
    {inputName=WorldGridExtents, input {expression=&self:MaterialExpressionReroute_6}}
    {inputName=VoxelSize, input {expression=&self:MaterialExpressionReroute_7}}
    {inputName=WorldToLocal0, input {expression=&self:MaterialExpressionReroute_1}}
    {inputName=WorldToLocal1, input {expression=&self:MaterialExpressionReroute_2}}
    {inputName=WorldToLocal2, input {expression=&self:MaterialExpressionReroute_3}}
    {inputName=WorldToLocal3, input {expression=&self:MaterialExpressionReroute_4}}
    {inputName=CameraDirectionVector, input {expression=&self:MaterialExpressionMaterialFunctionCall_1}}
    {inputName=SceneDepth, input {expression=&self:MaterialExpressionSceneDepth_0, mask=1, maskR=1}}
    {inputName=DebugBBox, input {expression=&self:MaterialExpressionScalarParameter_4}}
    {
     inputName=Center
     input {expression=&self:MaterialExpressionVectorParameter_1, mask=1, maskB=1, maskG=1, maskR=1}
    }
    {
     inputName=Normal
     input {expression=&self:MaterialExpressionVectorParameter_2, mask=1, maskB=1, maskG=1, maskR=1}
    }
    {inputName=Tolerance, input {expression=&self:MaterialExpressionScalarParameter_1}}
    {inputName=NumSteps, input {expression=&self:MaterialExpressionScalarParameter_2}}
    {inputName=ComputeNormals, input {expression=&self:MaterialExpressionScalarParameter_3}}
   ]
  }
  ComponentMask(None=@n6) {a=true}
  n7 = Constant()
  n8 = Constant() {r=1}
  n9 = Constant() {r=-1}
  DotProduct(
   A=Multiply(A=MaterialFunctionCall(X=@n7, Y=@n8, Z=@n9) {materialFunction=&MakeFloat3}.Result) {constB=-1}
   B=ComponentMask(None=@n6) {b=true, g=true, r=true}
  )
  VectorParameter() {defaultValue=<rgba 1 1 1 0>}
 ]
}
