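# Translucent unlit debug material that draws a planar slice through a 3D volume texture, for Niagara Fluids grids.
# Custom HLSL intersects the camera ray with grid bounds and slice plane, rejects scene-occluded hits, then samples or loads voxels.
# It can instead draw checkered bounds, and overlays grid lines when the normal's component sum is at most one.
# A static switch picks legacy WorldToLocal matrix parameters or built-in local transforms with a pivot.
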Material /NiagaraFluids/Materials/Grid3D/M_DebugVolumeSlice
graph {
 properties {
  bDisableDepthTest=true
  bUsedWithNiagaraMeshParticles=true
  bUsedWithNiagaraSprites=true
  blendMode=BLEND_Translucent
  pixelDepthOffsetMode=PDOM_Legacy
  shadingModel=MSM_Unlit
  usedShadingModels=Unlit
 }
 nodes [
  n1 = CameraPositionWS()
  MaterialExpressionReroute_0 = Reroute(None=@n1)
  MaterialExpressionNormalize_0 = Normalize(VectorInput=Subtract(A=WorldPosition().XYZ, B=@n1))
  MaterialExpressionTextureObjectParameter_0 = TextureObjectParameter() {parameterName=VolumeTex, samplerType=SAMPLERTYPE_LinearColor, texture=/Engine/EngineDebugMaterials/VolumeToRender}
  MaterialExpressionReroute_5 = Reroute(None=@MaterialExpressionTextureObjectParameter_0)
  MaterialExpressionVectorParameter_0 = VectorParameter() {parameterName=WorldGridExtents}
  MaterialExpressionReroute_6 = Reroute(None=@MaterialExpressionVectorParameter_0.RGB)
  MaterialExpressionScalarParameter_0 = ScalarParameter() {defaultValue=0.25, parameterName=Voxel Size}
  MaterialExpressionReroute_7 = Reroute(None=@MaterialExpressionScalarParameter_0)
  n2 = VectorParameter() {parameterName=WorldToLocal0}
  MaterialExpressionReroute_1 = Reroute(None=AppendVector(A=@n2.RGB, B=@n2.RGB))
  n3 = VectorParameter() {parameterName=WorldToLocal1}
  MaterialExpressionReroute_2 = Reroute(None=AppendVector(A=@n3.RGB, B=@n3.RGB))
  n4 = VectorParameter() {parameterName=WorldToLocal2}
  MaterialExpressionReroute_3 = Reroute(None=AppendVector(A=@n4.RGB, B=@n4.RGB))
  n5 = VectorParameter() {parameterName=WorldToLocal3}
  MaterialExpressionReroute_4 = Reroute(None=AppendVector(A=@n5.RGB, B=@n5.RGB))
  MaterialExpressionMaterialFunctionCall_1 = MaterialFunctionCall() {materialFunction=&CameraDirectionVector}
  MaterialExpressionSceneDepth_0 = SceneDepth()
  MaterialExpressionScalarParameter_4 = ScalarParameter() {parameterName=Debug Bounds}
  MaterialExpressionVectorParameter_1 = VectorParameter() {parameterName=DebugSliceCenter, defaultValue=<rgba 0 0 100 0>}
  MaterialExpressionVectorParameter_2 = VectorParameter() {parameterName=DebugSliceAxis, defaultValue=<rgba 0 1 0 0>}
  MaterialExpressionScalarParameter_1 = ScalarParameter() {parameterName=Debug Thickness}
  MaterialExpressionScalarParameter_2 = ScalarParameter() {defaultValue=1, parameterName=Interpolate Grid}
  MaterialExpressionScalarParameter_3 = ScalarParameter() {parameterName=Index Space Slice}
  n6 = Custom!(
   RayStart=@MaterialExpressionReroute_0
   RayDir=@MaterialExpressionNormalize_0
   VolumeTexture=@MaterialExpressionReroute_5
   WorldGridExtents=@MaterialExpressionReroute_6
   VoxelSize=@MaterialExpressionReroute_7
   WorldToLocal0=@MaterialExpressionReroute_1
   WorldToLocal1=@MaterialExpressionReroute_2
   WorldToLocal2=@MaterialExpressionReroute_3
   WorldToLocal3=@MaterialExpressionReroute_4
   CameraDirectionVector=@MaterialExpressionMaterialFunctionCall_1.Result
   SceneDepth=@MaterialExpressionSceneDepth_0.R
   DebugBBox=@MaterialExpressionScalarParameter_4
   Center=@MaterialExpressionVectorParameter_1.RGB
   Normal=@MaterialExpressionVectorParameter_2.RGB
   DebugThickness=@MaterialExpressionScalarParameter_1
   InterpolateGrid=@MaterialExpressionScalarParameter_2
   IndexSpaceSlice=@MaterialExpressionScalarParameter_3
  ) {
   code="float4 ret = {0,0,0,0};\r\n\r\nfloat4x4 WorldToLocal = {\r\nWorldToLocal0.x, WorldToLocal0.y, WorldToLocal0.z, WorldToLocal0.w, \r\nWorldToLocal1.x, WorldToLocal1.y, WorldToLocal1.z, WorldToLocal1.w, \r\nWorldToLocal2.x, WorldToLocal2.y, WorldToLocal2.z, WorldToLocal2.w, \r\nWorldToLocal3.x, WorldToLocal3.y, WorldToLocal3.z, WorldToLocal3.w\r\n};\r\n\r\nfloat WorldDepth = SceneDepth / dot(RayDir, CameraDirectionVector);\r\n\r\nRayStart = mul(float4(RayStart,1), WorldToLocal).xyz;\r\nRayDir = mul(float4(RayDir,0), WorldToLocal).xyz;\r\n\r\n//\r\n//  Ray aabbox intersection\r\n//\r\n\r\nfloat3 dirfrac = 1.0f / RayDir;\r\n\r\nfloat t = 0;\r\n\r\nfloat3 MinBounds = -1.* WorldGridExtents * .5;\r\nfloat3 MaxBounds = WorldGridExtents * .5;\r\nfloat t1 = (MinBounds.x - RayStart.x)*dirfrac.x;\r\nfloat t2 = (MaxBounds.x - RayStart.x)*dirfrac.x;\r\nfloat t3 = (MinBounds.y - RayStart.y)*dirfrac.y;\r\nfloat t4 = (MaxBounds.y - RayStart.y)*dirfrac.y;\r\nfloat t5 = (MinBounds.z - RayStart.z)*dirfrac.z;\r\nfloat t6 = (MaxBounds.z - RayStart.z)*dirfrac.z;\r\n\r\nfloat tmin = max(max(min(t1, t2), min(t3, t4)), min(t5, t6));\r\nfloat tmax = min(min(max(t1, t2), max(t3, t4)), max(t5, t6));\r\n\r\n// if tmax < 0, ray (line) is intersecting AABB, but the whole AABB is behind us\r\nif (tmax < 0)\r\n{\r\n    t = tmax;\r\n    return ret;\r\n}\r\n\r\n// if tmin > tmax, ray doesn't intersect AABB\r\nif (tmin > tmax)\r\n{\r\n    t = tmax;\r\n    return ret;\r\n}\r\n\r\nt = tmin;\r\n\r\nfloat3 CurrLocalPos = RayStart + RayDir * t;\r\n\r\nif (DebugBBox > 1e-3 &&&& t < WorldDepth) \r\n{\r\n\tfloat4 rettmp = {1,1,1,DebugBBox};\r\n\r\n\tint3 VoxelCount = floor(CurrLocalPos / VoxelSize);\r\n\r\n\tint SliceParity = (VoxelCount.x) % 2;\r\n\tint RowParity = (VoxelCount.y + SliceParity+1) % 2;\r\n\tint CellParity = (VoxelCount.z + RowParity ) % 2;\r\n\r\n\tif (CellParity == 0)\r\n\t{\r\n\t\trettmp.y = 0;\r\n\t}\r\n\r\n\treturn rettmp;\r\n}\r\n\r\nint3 sz;\r\nVolumeTexture.GetDimensions(sz.x, sz.y, sz.z);\r\n\r\nif (IndexSpaceSlice)\r\n{\r\n    Center = (Center + .5) / sz;\r\n    Center = Center * WorldGridExtents + MinBounds;\r\n}\r\n\r\nfloat denom = dot(RayDir, Normal);\r\nif (abs(denom) > 1e-6)\r\n{\r\n\tt = dot(Center -RayStart, Normal) / denom;\r\n\tif (t > 1e-5)\r\n\t{\r\n\t\tCurrLocalPos = RayStart + RayDir * t;\r\n\t\tfloat3 UnitPos = (CurrLocalPos - MinBounds) / WorldGridExtents;\r\n\r\n\t\tif (UnitPos.x >= 0 &&&& UnitPos.x <=1 &&&&\r\n\t\t\tUnitPos.y >= 0 &&&& UnitPos.y <=1 &&&&\r\n\t\t\tUnitPos.z >= 0 &&&& UnitPos.z <= 1 &&&& t < WorldDepth)\r\n\t\t{\r\n\r\n\t\t\tfloat3 V = UnitPos * sz - .5;\r\n\r\n\t\t\tint3 V_int = round(V);\r\n\r\n         if (InterpolateGrid)\r\n         {\r\n             ret = VolumeTexture.SampleLevel(VolumeTextureSampler,  UnitPos, 0);\r\n         }\r\n         else\r\n         {\r\n\t\t\t    ret = VolumeTexture.Load(int4(V_int.x, V_int.y, V_int.z, 0));\r\n         }\r\n\t\t\tret.a = 1;\r\n\r\n        // if we are using a cardinal axis, then we can draw the grid lines\r\n        if (Normal.x+Normal.y+Normal.z <= 1)\r\n        {\r\n\t\t\t  float3 Remain = abs(V - V_int - .5);\r\n\t\t\t  float2 CenterTmp = dot(abs(V - V_int), 1.-Normal);\r\n           float3 DebugLines = dot(Remain, 1.-Normal);\r\n\t\t\t  if ((Remain.x < DebugThickness  || Remain.y < DebugThickness || Remain.z < DebugThickness) \r\n              || length(CenterTmp) < DebugThickness || \r\n              Remain.x > 1 - DebugThickness || Remain.y > 1 - DebugThickness || Remain.z > 1 - DebugThickness)\r\n\t\t\t  {\r\n\t\t\t\t  ret.rgb = 1;\r\n\t\t\t  }\r\n         }\r\n\r\n\t\t\treturn ret;\r\n\t\t}\r\n\t}\r\n}\r\nreturn float4(0,0,0,0);\r\n\r\n"
   outputType=CMOT_Float4
   inputs [
    {inputName=RayStart, input {expression=&self:MaterialExpressionReroute_0}}
    {inputName=RayDir, input {expression=&self:MaterialExpressionNormalize_0}}
    {inputName=VolumeTexture, input {expression=&self:MaterialExpressionReroute_5}}
    {inputName=WorldGridExtents, input {expression=&self:MaterialExpressionReroute_6}}
    {inputName=VoxelSize, input {expression=&self:MaterialExpressionReroute_7}}
    {inputName=WorldToLocal0, input {expression=&self:MaterialExpressionReroute_1}}
    {inputName=WorldToLocal1, input {expression=&self:MaterialExpressionReroute_2}}
    {inputName=WorldToLocal2, input {expression=&self:MaterialExpressionReroute_3}}
    {inputName=WorldToLocal3, input {expression=&self:MaterialExpressionReroute_4}}
    {
     inputName=CameraDirectionVector
     input {expression=&self:MaterialExpressionMaterialFunctionCall_1}
    }
    {inputName=SceneDepth, input {expression=&self:MaterialExpressionSceneDepth_0, mask=1, maskR=1}}
    {inputName=DebugBBox, input {expression=&self:MaterialExpressionScalarParameter_4}}
    {
     inputName=Center
     input {expression=&self:MaterialExpressionVectorParameter_1, mask=1, maskB=1, maskG=1, maskR=1}
    }
    {
     inputName=Normal
     input {expression=&self:MaterialExpressionVectorParameter_2, mask=1, maskB=1, maskG=1, maskR=1}
    }
    {inputName=DebugThickness, input {expression=&self:MaterialExpressionScalarParameter_1}}
    {inputName=InterpolateGrid, input {expression=&self:MaterialExpressionScalarParameter_2}}
    {inputName=IndexSpaceSlice, input {expression=&self:MaterialExpressionScalarParameter_3}}
   ]
  }
  MaterialExpressionTransformPosition_0 = TransformPosition(None=@n1) {transformSourceType=TRANSFORMPOSSOURCE_World}
  MaterialExpressionTransform_0 = Transform(None=@MaterialExpressionNormalize_0) {transformSourceType=TRANSFORMSOURCE_World, transformType=TRANSFORM_Local}
  MaterialExpressionVectorParameter_6 = VectorParameter() {parameterName=LocalPivot}
  n7 = Custom!(
   RayStart=@MaterialExpressionTransformPosition_0
   RayDir=@MaterialExpressionTransform_0
   VolumeTexture=@MaterialExpressionTextureObjectParameter_0
   WorldGridExtents=@MaterialExpressionVectorParameter_0.RGB
   VoxelSize=@MaterialExpressionScalarParameter_0
   CameraDirectionVector=@MaterialExpressionMaterialFunctionCall_1.Result
   SceneDepth=@MaterialExpressionSceneDepth_0.R
   DebugBBox=@MaterialExpressionScalarParameter_4
   Center=@MaterialExpressionVectorParameter_1.RGB
   Normal=@MaterialExpressionVectorParameter_2.RGB
   DebugThickness=@MaterialExpressionScalarParameter_1
   InterpolateGrid=@MaterialExpressionScalarParameter_2
   IndexSpaceSlice=@MaterialExpressionScalarParameter_3
   LocalPivot=@MaterialExpressionVectorParameter_6.RGB
  ) {
   code="float4 ret = {0,0,0,0};\r\n\r\nfloat WorldDepth = SceneDepth / dot(RayDir, CameraDirectionVector);\r\n//\r\n//  Ray aabbox intersection\r\n//\r\n\r\nfloat3 dirfrac = 1.0f / RayDir;\r\n\r\nfloat t = 0;\r\n\r\nfloat3 MinBounds = -1.* WorldGridExtents * .5 + (.5 - LocalPivot)*WorldGridExtents;\r\nfloat3 MaxBounds = WorldGridExtents * .5 + (.5 - LocalPivot)*WorldGridExtents;\r\nfloat t1 = (MinBounds.x - RayStart.x)*dirfrac.x;\r\nfloat t2 = (MaxBounds.x - RayStart.x)*dirfrac.x;\r\nfloat t3 = (MinBounds.y - RayStart.y)*dirfrac.y;\r\nfloat t4 = (MaxBounds.y - RayStart.y)*dirfrac.y;\r\nfloat t5 = (MinBounds.z - RayStart.z)*dirfrac.z;\r\nfloat t6 = (MaxBounds.z - RayStart.z)*dirfrac.z;\r\n\r\nfloat tmin = max(max(min(t1, t2), min(t3, t4)), min(t5, t6));\r\nfloat tmax = min(min(max(t1, t2), max(t3, t4)), max(t5, t6));\r\n\r\n// if tmax < 0, ray (line) is intersecting AABB, but the whole AABB is behind us\r\nif (tmax < 0)\r\n{\r\n    t = tmax;\r\n    return ret;\r\n}\r\n\r\n// if tmin > tmax, ray doesn't intersect AABB\r\nif (tmin > tmax)\r\n{\r\n    t = tmax;\r\n    return ret;\r\n}\r\n\r\nt = tmin;\r\n\r\nfloat3 CurrLocalPos = RayStart + RayDir * t;\r\n\r\nif (DebugBBox > 1e-3 &&&& t < WorldDepth) \r\n{\r\n\tfloat4 rettmp = {1,1,1,DebugBBox};\r\n\r\n\tint3 VoxelCount = floor(CurrLocalPos / VoxelSize);\r\n\r\n\tint SliceParity = (VoxelCount.x) % 2;\r\n\tint RowParity = (VoxelCount.y + SliceParity+1) % 2;\r\n\tint CellParity = (VoxelCount.z + RowParity ) % 2;\r\n\r\n\tif (CellParity == 0)\r\n\t{\r\n\t\trettmp.y = 0;\r\n\t}\r\n\r\n\treturn rettmp;\r\n}\r\n\r\nint3 sz;\r\nVolumeTexture.GetDimensions(sz.x, sz.y, sz.z);\r\n\r\nif (IndexSpaceSlice)\r\n{\r\n    Center = (Center + .5) / sz;\r\n    Center = Center * WorldGridExtents + MinBounds;\r\n}\r\n\r\nfloat denom = dot(RayDir, Normal);\r\nif (abs(denom) > 1e-6)\r\n{\r\n\tt = dot(Center -RayStart, Normal) / denom;\r\n\tif (t > 1e-5)\r\n\t{\r\n\t\tCurrLocalPos = RayStart + RayDir * t;\r\n\t\tfloat3 UnitPos = CurrLocalPos / WorldGridExtents + LocalPivot;\r\n\r\n\t\tif (UnitPos.x >= 0 &&&& UnitPos.x <=1 &&&&\r\n\t\t\tUnitPos.y >= 0 &&&& UnitPos.y <=1 &&&&\r\n\t\t\tUnitPos.z >= 0 &&&& UnitPos.z <= 1 &&&& t < WorldDepth)\r\n\t\t{\r\n\r\n\t\t\tfloat3 V = UnitPos * sz - .5;\r\n\r\n\t\t\tint3 V_int = round(V);\r\n\r\n         if (InterpolateGrid)\r\n         {\r\n             ret = VolumeTexture.SampleLevel(VolumeTextureSampler,  UnitPos, 0);\r\n         }\r\n         else\r\n         {\r\n\t\t\t    ret = VolumeTexture.Load(int4(V_int.x, V_int.y, V_int.z, 0));\r\n         }\r\n\t\t\tret.a = 1;\r\n\r\n        // if we are using a cardinal axis, then we can draw the grid lines\r\n        if (Normal.x+Normal.y+Normal.z <= 1)\r\n        {\r\n\t\t\t  float3 Remain = abs(V - V_int - .5);\r\n\t\t\t  float2 CenterTmp = dot(abs(V - V_int), 1.-Normal);\r\n           float3 DebugLines = dot(Remain, 1.-Normal);\r\n\t\t\t  if ((Remain.x < DebugThickness  || Remain.y < DebugThickness || Remain.z < DebugThickness) \r\n              || length(CenterTmp) < DebugThickness || \r\n              Remain.x > 1 - DebugThickness || Remain.y > 1 - DebugThickness || Remain.z > 1 - DebugThickness)\r\n\t\t\t  {\r\n\t\t\t\t  ret.rgb = 1;\r\n\t\t\t  }\r\n         }\r\n\r\n\t\t\treturn ret;\r\n\t\t}\r\n\t}\r\n}\r\nreturn float4(0,0,0,0);\r\n\r\n"
   outputType=CMOT_Float4
   inputs [
    {inputName=RayStart, input {expression=&self:MaterialExpressionTransformPosition_0}}
    {inputName=RayDir, input {expression=&self:MaterialExpressionTransform_0}}
    {inputName=VolumeTexture, input {expression=&self:MaterialExpressionTextureObjectParameter_0}}
    {
     inputName=WorldGridExtents
     input {expression=&self:MaterialExpressionVectorParameter_0, mask=1, maskB=1, maskG=1, maskR=1}
    }
    {inputName=VoxelSize, input {expression=&self:MaterialExpressionScalarParameter_0}}
    {
     inputName=CameraDirectionVector
     input {expression=&self:MaterialExpressionMaterialFunctionCall_1}
    }
    {inputName=SceneDepth, input {expression=&self:MaterialExpressionSceneDepth_0, mask=1, maskR=1}}
    {inputName=DebugBBox, input {expression=&self:MaterialExpressionScalarParameter_4}}
    {
     inputName=Center
     input {expression=&self:MaterialExpressionVectorParameter_1, mask=1, maskB=1, maskG=1, maskR=1}
    }
    {
     inputName=Normal
     input {expression=&self:MaterialExpressionVectorParameter_2, mask=1, maskB=1, maskG=1, maskR=1}
    }
    {inputName=DebugThickness, input {expression=&self:MaterialExpressionScalarParameter_1}}
    {inputName=InterpolateGrid, input {expression=&self:MaterialExpressionScalarParameter_2}}
    {inputName=IndexSpaceSlice, input {expression=&self:MaterialExpressionScalarParameter_3}}
    {
     inputName=LocalPivot
     input {expression=&self:MaterialExpressionVectorParameter_6, mask=1, maskB=1, maskG=1, maskR=1}
    }
   ]
  }
  n8 = StaticSwitchParameter(True=@n6, False=@n7) {parameterName=Use Legacy Transform}
  ComponentMask(None=@n8) {b=true, g=true, r=true}
  ComponentMask(None=@n8) {a=true}
 ]
}
