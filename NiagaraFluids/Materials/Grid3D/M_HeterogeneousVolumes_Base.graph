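# Base additive volume-domain material for rendering Niagara Fluids 3D grid data, such as density and temperature, as Heterogeneous Volumes.
# World position, transformed by WorldToLocal parameters and normalized by grid extents, samples VolumeTex for density (R) and temperature (B).
# Both channels are remapped by offset, range, gain and optional ramp textures, then combined into an extinction-style sum with constant extinction added where density is positive.
# Emission color comes from a blackbody or temperature ramp, scaled by gains.
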Material /NiagaraFluids/Materials/Grid3D/M_HeterogeneousVolumes_Base
graph {
 properties {
  bUsedWithHeterogeneousVolumes=true
  bUsedWithNiagaraMeshParticles=true
  blendMode=BLEND_Additive
  materialDomain=MD_Volume
  pixelDepthOffsetMode=PDOM_Legacy
  usedShadingModels=Default Lit
 }
 nodes [
  n1 = ScalarParameter() {defaultValue=1, parameterName=EmissionGain}
  n2 = ScalarParameter() {parameterName=UseTemperatureOpacity'Curve}
  n3 = ScalarParameter() {parameterName=TemperatureRange}
  n4 = MaterialFunctionCall(
   VectorToTransform=WorldPosition().XYZ
   BasisX=VectorParameter() {parameterName=WorldToLocal0}.RGB
   BasisY=VectorParameter() {parameterName=WorldToLocal1}.RGB
   BasisZ=VectorParameter() {parameterName=WorldToLocal2}.RGB
   BasisPosition=VectorParameter() {parameterName=WorldToLocal3}.RGB
  ) {materialFunction=&Transform3x3Matrix}
  n5 = VectorParameter() {parameterName=WorldGridExtents}
  n6 = Divide(A=@n5.RGB)
  n7 = TextureSampleParameterVolume!(UVs=Divide(A=Add(A=@n4.Transformed Vector, B=@n6), B=@n5.RGB), Apply View MipBias) {parameterName=VolumeTex}
  n8 = ScalarParameter() {parameterName=TemperatureOffset}
  n9 = Constant()
  n10 = If(A=@n3, B=Constant() {r=0}, "A > B"=Divide(A=Subtract(A=@n7.B, B=@n8), B=@n3), "A == B"=@n9, "A < B"=@n9)
  n11 = TextureSampleParameter2D(UVs=@n10) {parameterName=TemperatureOpacityRamp, samplerType=SAMPLERTYPE_LinearColor, texture=/Engine/ArtTools/RenderToTexture/Textures/127grey}
  n12 = TextureSampleParameter2D(UVs=@n10) {parameterName=TemperatureRamp, samplerType=SAMPLERTYPE_LinearColor, texture=&127grey}
  n13 = If(A=@n2, "A > B"=@n11.R, "A == B"=@n10, "A < B"=@n10)
  n14 = If(
   A=ScalarParameter() {parameterName=UseTemperatureCurve}
   "A > B"=If(A=@n2, "A > B"=@n11.R, "A == B"=@n12.A, "A < B"=@n12.A)
   "A == B"=@n13
   "A < B"=@n13
  )
  n15 = ScalarParameter() {defaultValue=1, parameterName=TemperatureOpacityGain}
  n16 = ScalarParameter() {parameterName=DensityOffset}
  n17 = ScalarParameter() {parameterName=DensityRange}
  n18 = Constant() {r=1}
  n19 = Saturate(None=Divide(A=Subtract(A=@n7.R, B=@n16), B=If(A=@n17, "A > B"=@n17, "A == B"=@n18, "A < B"=@n18)))
  n20 = Multiply(
   A=If(
    A=ScalarParameter() {parameterName=UseDensityCurve}
    "A > B"=TextureSampleParameter2D(UVs=@n19) {parameterName=DensityRamp, samplerType=SAMPLERTYPE_LinearColor, texture=&127grey}.R
    "A == B"=@n19
    "A < B"=@n19
   )
   B=ScalarParameter() {defaultValue=1, parameterName=DensityGain}
  )
  n21 = ScalarParameter() {parameterName=ConstantExtinction}
  Add(A=Multiply(A=@n1, B=Multiply(A=@n14, B=@n15)), B=If(A=@n20, "A > B"=Add(A=@n20, B=@n21), "A == B"=@n20, "A < B"=@n20) {constB=0})
  n22 = MaterialFunctionCall() {materialFunction=&LocalPosition}
  n23 = MaterialFunctionCall() {materialFunction=&ObjectLocalBounds}
  Divide(A=Subtract(A=Add(A=@n22.Local Position) {constB=0}, B=@n23.Local Bounds Minimum), B=@n23.Local Bounds Size)
  Multiply(A=ScalarParameter() {defaultValue=1, parameterName=Albedo}, B=ScalarParameter() {defaultValue=1, parameterName=AlbedoGain})
  n24 = BlackBody!(Temp=Multiply(A=Constant() {r=4500}, B=@n10))
  n25 = ScalarParameter() {defaultValue=1, parameterName=TemperatureGain}
  n26 = Multiply(A=@n12.RGB, B=@n25)
  Multiply(
   A=If(A=ScalarParameter() {defaultValue=1, parameterName=UseBlackbodyCurve}, "A > B"=Multiply(A=@n24, B=@n25), "A == B"=@n26, "A < B"=@n26)
   B=@n1
  )
  ScalarParameter() {parameterName=UseColorTexture}
  TextureSampleParameterVolume() {parameterName=ColorTexture}
 ]
}
