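# Material function that sets up a camera ray march through a Niagara 3D grid volume.
# A Custom HLSL node transforms the ray into 0-1 grid space, intersects the box, clips at scene depth and derives step sizes and a capped count from voxel size.
# The entry point optionally snaps to step intervals, then gets per-pixel random jitter.
# The Trace Distance input has no effect because its clamp is commented out.
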MaterialFunction /NiagaraFluids/Materials/Functions/Niagara_Grid3D_BoxIntersection
graph {nodes [
  MaterialExpressionCameraPositionWS_0 = CameraPositionWS()
  MaterialExpressionNormalize_0 = Normalize(VectorInput=Subtract(A=WorldPosition().XYZ, B=@MaterialExpressionCameraPositionWS_0))
  MaterialExpressionVectorParameter_9 = VectorParameter() {group=NiagaraControlled, parameterName=WorldGridExtents, defaultValue=<rgba 256 256 256 0>}
  n1 = VectorParameter() {group=NiagaraControlled, parameterName=WorldToLocal0, defaultValue=<rgba 1 0 0 0>}
  MaterialExpressionReroute_0 = Reroute(None=AppendVector(A=@n1.RGB, B=@n1.RGB))
  n2 = VectorParameter() {group=NiagaraControlled, parameterName=WorldToLocal1, defaultValue=<rgba 0 1 0 0>}
  MaterialExpressionAppendVector_5 = AppendVector(A=@n2.RGB, B=@n2.RGB)
  n3 = VectorParameter() {group=NiagaraControlled, parameterName=WorldToLocal2, defaultValue=<rgba 0 0 1 0>}
  MaterialExpressionReroute_1 = Reroute(None=AppendVector(A=@n3.RGB, B=@n3.RGB))
  n4 = VectorParameter() {group=NiagaraControlled, parameterName=WorldToLocal3}
  MaterialExpressionAppendVector_7 = AppendVector(A=@n4.RGB, B=@n4.RGB)
  MaterialExpressionSceneDepth_1 = SceneDepth()
  MaterialExpressionMaterialFunctionCall_1 = MaterialFunctionCall() {materialFunction=&CameraDirectionVector}
  MaterialExpressionFunctionInput_1 = FunctionInput() {bUsePreviewValueAsDefault=true, inputName=Trace Distance, inputType=FunctionInput_Scalar, sortPriority=1, previewValue=<1 0 0 1>}
  MaterialExpressionFunctionInput_0 = FunctionInput() {bUsePreviewValueAsDefault=true, inputName=Step Size Scale, inputType=FunctionInput_Scalar, sortPriority=1, previewValue=<1 0 0 1>}
  MaterialExpressionScalarParameter_2 = ScalarParameter() {defaultValue=8, group=NiagaraControlled, parameterName=Voxel Size}
  MaterialExpressionScalarParameter_0 = ScalarParameter() {defaultValue=1, group=Raymarch Setup, parameterName=Snap Ray to Plane Interval}
  MaterialExpressionCustom_0 = Custom!(
   RayStart=@MaterialExpressionCameraPositionWS_0
   RayDir=@MaterialExpressionNormalize_0
   WorldGridExtents=@MaterialExpressionVectorParameter_9.RGB
   WorldToLocal0=@MaterialExpressionReroute_0
   WorldToLocal1=@MaterialExpressionAppendVector_5
   WorldToLocal2=@MaterialExpressionReroute_1
   WorldToLocal3=@MaterialExpressionAppendVector_7
   SceneDepth=@MaterialExpressionSceneDepth_1.R
   CameraDirectionVector=@MaterialExpressionMaterialFunctionCall_1.Result
   TraceDistance=@MaterialExpressionFunctionInput_1
   StepSizeMultiplier=@MaterialExpressionFunctionInput_0
   VoxelSize=@MaterialExpressionScalarParameter_2
   TraceViewAlignedSamples=@MaterialExpressionScalarParameter_0
  ) {
   code="##define MAX_SAFE_STEPS 1024\r\n\r\nfloat4x4 WorldToLocal = {\r\nWorldToLocal0.x, WorldToLocal0.y, WorldToLocal0.z, WorldToLocal0.w, \r\nWorldToLocal1.x, WorldToLocal1.y, WorldToLocal1.z, WorldToLocal1.w, \r\nWorldToLocal2.x, WorldToLocal2.y, WorldToLocal2.z, WorldToLocal2.w, \r\nWorldToLocal3.x, WorldToLocal3.y, WorldToLocal3.z, WorldToLocal3.w\r\n};\r\n\r\nfloat maxaxis = max(WorldGridExtents.x, max(WorldGridExtents.y, WorldGridExtents.z));\r\nfloat camthetacos = abs(dot(normalize(RayDir), normalize(CameraDirectionVector)));\r\n\r\n//derive step count from extents and voxelsize\r\nfloat MaxAxisRes = maxaxis / VoxelSize;\r\n\r\n//adjust step count for stepsize multiplier\r\nMaxAxisRes /= StepSizeMultiplier;\r\n\r\n//Derive the number of Steps needed to trace a full axis\r\nfloat stepsize = (1.0 / MaxAxisRes);\r\n\r\nfloat WorldDepth = SceneDepth / camthetacos;\r\n\r\n//convert depth into Unit space of the grid\r\nfloat3 depthpos = (Parameters.CameraVector * SceneDepth) / camthetacos;\r\ndepthpos = mul(float4(depthpos,0),  WorldToLocal).xyz;\r\ndepthpos /= WorldGridExtents;\r\n\r\n//Get the total ray length to depth in unit space\r\nfloat unitscenedepth = length(depthpos);\r\n\r\nRayStart = mul(float4(RayStart,1), WorldToLocal).xyz;\r\nRayDir = normalize(mul(float4(RayDir,0), WorldToLocal).xyz);\r\n\r\n//Calculate the local scale of a voxel in unit space\r\nfloat VoxelLocalToUnitScale = length(RayDir / (WorldGridExtents/maxaxis));\r\n\r\nRayStart = (RayStart / WorldGridExtents) + 0.5;\r\nRayDir = normalize(RayDir / WorldGridExtents);\r\nfloat planecos = abs(dot(normalize(RayStart - 0.5), RayDir));\r\n\r\n//Disabling the Uniform Depth option for now due to issues finding and ideal plane orientation for all cases\r\n//if(TraceUniformDepth > 0)\r\n{\r\n\t//stepsize /= camthetacos;\r\n}\r\n\r\n//Adjust the step size to be in unit space\r\nstepsize *= VoxelLocalToUnitScale;\r\n\r\n//  Ray aabbox intersection\r\n\r\nfloat3 invraydir = 1.0f / RayDir;\r\nfloat3 firstintersections = (0 - RayStart) * invraydir;\r\nfloat3 secondintersections = (1 - RayStart) * invraydir;\r\nfloat3 closest = min(firstintersections, secondintersections);\r\nfloat3 furthest = max(firstintersections, secondintersections);\r\n\r\nfloat tmin = max(closest.x, max(closest.y, closest.z));\r\nfloat tmax = min(furthest.x, min(furthest.y, furthest.z));\r\ntmin = max(0, tmin);\r\n\r\nif(TraceViewAlignedSamples > 0)\r\n{\r\n\tfloat dist_to_plane_center = length(RayStart - 0.5) / 1;\r\n\tfloat dist_box_to_plane = dist_to_plane_center - tmin;\r\n\tfloat steps_to_plane = dist_box_to_plane / stepsize;\r\n\tfloat snap_dist = frac(steps_to_plane);\r\n\t\r\n\ttmin += snap_dist * stepsize;\r\n}\r\n\r\n\r\n//Limit distance by user specified trace distance\r\n//TraceDistance /= max(abs(RayDir.x), abs(RayDir.y));\r\n//tmax = min(TraceDistance, tmax - tmin) + tmin;\r\n\t\r\ntmax = min(tmax, unitscenedepth);\r\n\r\nfloat thickness = max(0, tmax - tmin);\r\n\r\nCalculatedSteps = min((thickness/stepsize), MAX_SAFE_STEPS);\r\n\r\nRayOrigin = RayStart + (RayDir * tmin);\r\nRayDirection = RayDir;\r\n\r\nUnitStepSize = stepsize;\r\nLocalStepSize = VoxelSize * StepSizeMultiplier;\r\n\r\n\r\nreturn 0;\r\n"
   outputType=CMOT_Float1
   additionalOutputs [
    {outputName=RayOrigin, outputType=CMOT_Float3}
    {outputName=RayDirection, outputType=CMOT_Float3}
    {outputName=CalculatedSteps, outputType=CMOT_Float1}
    {outputName=UnitStepSize, outputType=CMOT_Float1}
    {outputName=LocalStepSize, outputType=CMOT_Float1}
   ]
   inputs [
    {
     inputName=RayStart
     input {expression=&self:MaterialExpressionCameraPositionWS_0}
    }
    {inputName=RayDir, input {expression=&self:MaterialExpressionNormalize_0}}
    {
     inputName=WorldGridExtents
     input {
      expression=&self:MaterialExpressionVectorParameter_9
      mask=1
      maskB=1
      maskG=1
      maskR=1
     }
    }
    {inputName=WorldToLocal0, input {expression=&self:MaterialExpressionReroute_0}}
    {
     inputName=WorldToLocal1
     input {expression=&self:MaterialExpressionAppendVector_5}
    }
    {inputName=WorldToLocal2, input {expression=&self:MaterialExpressionReroute_1}}
    {
     inputName=WorldToLocal3
     input {expression=&self:MaterialExpressionAppendVector_7}
    }
    {
     inputName=SceneDepth
     input {expression=&self:MaterialExpressionSceneDepth_1, mask=1, maskR=1}
    }
    {
     inputName=CameraDirectionVector
     input {expression=&self:MaterialExpressionMaterialFunctionCall_1}
    }
    {
     inputName=TraceDistance
     input {expression=&self:MaterialExpressionFunctionInput_1}
    }
    {
     inputName=StepSizeMultiplier
     input {expression=&self:MaterialExpressionFunctionInput_0}
    }
    {
     inputName=VoxelSize
     input {expression=&self:MaterialExpressionScalarParameter_2}
    }
    {
     inputName=TraceViewAlignedSamples
     input {expression=&self:MaterialExpressionScalarParameter_0}
    }
   ]
  }
  MaterialExpressionScalarParameter_1 = ScalarParameter() {defaultValue=1, parameterName=RandomRayStartOffset}
  MaterialExpressionScreenPosition_0 = ScreenPosition()
  n5 = Custom!(
   RayStart=@MaterialExpressionCustom_0.RayOrigin
   RayDir=@MaterialExpressionCustom_0.RayOrigin
   UnitStepSize=@MaterialExpressionCustom_0.RayOrigin
   RandStartOffset=@MaterialExpressionScalarParameter_1
   PixelPos=@MaterialExpressionScreenPosition_0.PixelPosition
  ) {
   code="return RayStart + RayDir * UnitStepSize * RandStartOffset *PseudoRandom(PixelPos.xy);"
   inputs [
    {
     inputName=RayStart
     input {expression=&self:MaterialExpressionCustom_0, outputIndex=1}
    }
    {
     inputName=RayDir
     input {expression=&self:MaterialExpressionCustom_0, outputIndex=2}
    }
    {
     inputName=UnitStepSize
     input {expression=&self:MaterialExpressionCustom_0, outputIndex=4}
    }
    {
     inputName=RandStartOffset
     input {expression=&self:MaterialExpressionScalarParameter_1}
    }
    {
     inputName=PixelPos
     input {expression=&self:MaterialExpressionScreenPosition_0, outputIndex=1}
    }
   ]
  }
  FunctionOutput(None=@n5) {description="Ray Origin in Unit Space, 0-1", outputName=Ray Origin}
  FunctionOutput(None=@MaterialExpressionCustom_0.RayDirection) {description=Ray Direction in Unit Space, outputName=Ray Direction, sortPriority=1}
  FunctionOutput(None=@MaterialExpressionCustom_0.CalculatedSteps) {
   description=The number of steps that fit within the box intersection. It includes a fractional part used to take a fractional step at the end of whole steps.
   outputName=Calculated Steps
   sortPriority=2
  }
  FunctionOutput(None=@MaterialExpressionCustom_0.LocalStepSize) {
   description="The local step size is the size of each raymarch step in local space, which accounts for scale. This is used to calculate extinction coefficients."
   outputName=Local Step Size
   sortPriority=4
  }
  FunctionOutput(None=@MaterialExpressionCustom_0.UnitStepSize) {
   description="The unit step size gives the step size needed to step completely through the local 0-1 range, independent of non uniform scale."
   outputName=Unit Step Size
   sortPriority=3
  }
 ]}
