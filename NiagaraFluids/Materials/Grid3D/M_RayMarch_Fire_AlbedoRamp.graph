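# Unlit alpha-composite Niagara sprite material that raymarches a Grid3D fire volume.
# A box-intersection function feeds custom HLSL mapping density to extinction and temperature to emission through ramps, plus sun/fill lighting.
# Use Texture Advection selects noise-modulated density from two blended coordinate volumes; otherwise an albedo ramp, indexed by its own volume, tints lighting.
# Low quality, forward, mobile and ES paths skip marching; color optionally cancels eye adaptation.
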Material /NiagaraFluids/Materials/Grid3D/M_RayMarch_Fire_AlbedoRamp
graph {
 properties {
  bDisableDepthTest=true
  bUsedWithNiagaraSprites=true
  blendMode=BLEND_AlphaComposite
  pixelDepthOffsetMode=PDOM_Legacy
  shadingModel=MSM_Unlit
  usedShadingModels=Unlit
 }
 nodes [
  MaterialExpressionTextureObjectParameter_2 = TextureObjectParameter() {parameterName=VolumeTex, samplerType=SAMPLERTYPE_LinearColor, texture=/Engine/EngineDebugMaterials/VolumeToRender}
  MaterialExpressionMaterialFunctionCall_2 = MaterialFunctionCall!(
   Trace Distance=ScalarParameter() {defaultValue=1, group=Raymarch Setup, parameterName=TraceDistance}
   Step Size Scale=ScalarParameter() {defaultValue=1, group=Raymarch Setup, parameterName=Step Size Mult}
  ) {materialFunction=/NiagaraFluids/Materials/Functions/Niagara_Grid3D_BoxIntersection}
  n1 = VectorParameter() {parameterName=SunColor, defaultValue=<rgba 1 1 1 1>}
  MaterialExpressionMultiply_1 = Multiply(A=@n1.RGB, B=@n1.RGB)
  n2 = VectorParameter() {parameterName=FillColor, defaultValue=<rgba 1 1 1 1>}
  MaterialExpressionMultiply_5 = Multiply(A=@n2.RGB, B=@n2.RGB)
  MaterialExpressionScalarParameter_6 = ScalarParameter() {parameterName=ConstantExtinction}
  MaterialExpressionScalarParameter_17 = ScalarParameter() {parameterName=DensityOffset}
  MaterialExpressionScalarParameter_18 = ScalarParameter() {parameterName=DensityRange}
  MaterialExpressionTextureObjectParameter_4 = TextureObjectParameter() {group=Temperature Color Curve, parameterName=DensityRamp, texture=/Engine/EngineMaterials/HairDebugColor}
  MaterialExpressionScalarParameter_19 = ScalarParameter() {defaultValue=1, parameterName=DensityGain}
  MaterialExpressionScalarParameter_15 = ScalarParameter() {parameterName=TemperatureOffset}
  MaterialExpressionScalarParameter_16 = ScalarParameter() {parameterName=TemperatureRange}
  MaterialExpressionTextureObjectParameter_1 = TextureObjectParameter() {group=Temperature Color Curve, parameterName=TemperatureRamp, texture=&HairDebugColor}
  MaterialExpressionScalarParameter_5 = ScalarParameter() {defaultValue=1, parameterName=Albedo}
  MaterialExpressionTextureObjectParameter_6 = TextureObjectParameter() {
   parameterName=Noise Volume Texture
   samplerType=SAMPLERTYPE_LinearColor
   texture=/NiagaraFluids/Textures/VolumeTextures/periodic_noise_Volume
  }
  MaterialExpressionTextureObjectParameter_3 = TextureObjectParameter() {parameterName=TexCoord1, samplerType=SAMPLERTYPE_LinearColor, texture=&VolumeToRender}
  MaterialExpressionTextureObjectParameter_5 = TextureObjectParameter() {parameterName=TexCoord2, samplerType=SAMPLERTYPE_LinearColor, texture=&VolumeToRender}
  MaterialExpressionScalarParameter_8 = ScalarParameter() {parameterName=TexBlend}
  MaterialExpressionScalarParameter_12 = ScalarParameter() {defaultValue=0.2, parameterName=TexScale}
  MaterialExpressionScalarParameter_13 = ScalarParameter() {defaultValue=0.1, parameterName=TexFit_0}
  MaterialExpressionScalarParameter_21 = ScalarParameter() {defaultValue=0.9, parameterName=TexFit_1}
  n3 = Custom!(
   VolumeTexture=@MaterialExpressionTextureObjectParameter_2
   RayStart=@MaterialExpressionMaterialFunctionCall_2.Ray Origin
   RayDir=@MaterialExpressionMaterialFunctionCall_2.Ray Origin
   NumSteps=@MaterialExpressionMaterialFunctionCall_2.Ray Origin
   UnitStepSize=@MaterialExpressionMaterialFunctionCall_2.Ray Origin
   LocalStepSize=@MaterialExpressionMaterialFunctionCall_2.Ray Origin
   SunColor=@MaterialExpressionMultiply_1
   FillColor=@MaterialExpressionMultiply_5
   ConstantExtinction=@MaterialExpressionScalarParameter_6
   DensityOffset=@MaterialExpressionScalarParameter_17
   DensityRange=@MaterialExpressionScalarParameter_18
   DensityRamp=@MaterialExpressionTextureObjectParameter_4
   DensityGain=@MaterialExpressionScalarParameter_19
   TemperatureOffset=@MaterialExpressionScalarParameter_15
   TemperatureRange=@MaterialExpressionScalarParameter_16
   TemperatureRamp=@MaterialExpressionTextureObjectParameter_1
   Albedo=@MaterialExpressionScalarParameter_5
   NoiseTexture=@MaterialExpressionTextureObjectParameter_6
   TexCoord1=@MaterialExpressionTextureObjectParameter_3
   TexCoord2=@MaterialExpressionTextureObjectParameter_5
   TexBlend=@MaterialExpressionScalarParameter_8
   TexScale=@MaterialExpressionScalarParameter_12
   TexFit_0=@MaterialExpressionScalarParameter_13
   TexFit_1=@MaterialExpressionScalarParameter_21
  ) {
   code="struct FireRampsRenderVolume\r\n{\r\n\tTexture3D VolumeTexture;\r\n\tSamplerState VolumeTextureSampler;\r\n\tfloat Albedo;\r\n\r\n\tfloat3 SunColor;\r\n\tfloat3 FillColor;\r\n\tfloat ConstantExtinction;\r\n    float DensityOffset;\r\n    float DensityRange;\r\n    Texture2D DensityRamp;\r\n\tSamplerState DensityRampSampler;\r\n    float DensityGain;\r\n    float TemperatureOffset;\r\n    float TemperatureRange;\r\n    Texture2D TemperatureRamp;\r\n\tSamplerState TemperatureRampSampler;\r\n    Texture3D NoiseTexture;\r\n    SamplerState NoiseTextureSampler;\r\n    Texture3D TexCoord1;\r\n    SamplerState TexCoord1Sampler;\r\n    Texture3D TexCoord2;\r\n    SamplerState TexCoord2Sampler;\r\n    float TexBlend;\r\n    float TexScale;\r\n    float TexFit_0;\r\n    float TexFit_1;\r\n\r\n    float TextureScale;\r\n    float FitMin;\r\n    float FitRange;\r\n\r\n\r\n\t// Computes volume extinction coefficient and luminance for a given point in the volume\r\n\tvoid ComputeStep(float3 Position, float3 RayDirection, out float OutExtinction, out float OutScattering, out float3 OutLuminance)\r\n\t{  \r\n\r\n    float4 VolumeSample = VolumeTexture.SampleLevel(VolumeTextureSampler, Position, 0);\r\n\t\r\n    float Density = DensityRange > 1e-8 ? saturate((VolumeSample.r - DensityOffset) / DensityRange) : 0.0;\r\n\tfloat DensityCurve = DensityRamp.SampleLevel(DensityRampSampler, Density, 0 ).r;\r\n\t\r\n\r\n    float4 uvwa1 = TexCoord1.SampleLevel(TexCoord1Sampler, Position, 0);\r\n\tfloat3 uvw1  = frac(uvwa1.rgb * TextureScale);\r\n\tfloat Noise1 = NoiseTexture.SampleLevel(NoiseTextureSampler, uvw1, 0).r;\r\n\r\n    float4 uvwa2 = TexCoord2.SampleLevel(TexCoord2Sampler, Position, 0);\r\n\tfloat3 uvw2  = frac(uvwa2.rgb * TextureScale);\r\n\tfloat Noise2 = Texture3DSample(NoiseTexture, NoiseTextureSampler, uvw2).r;\r\n\t\r\n\tfloat Noise = lerp(Noise1, Noise2, TexBlend);\r\n\tNoise = saturate((Noise - FitMin) / FitRange);\r\n\r\n    OutExtinction = DensityCurve * DensityGain * lerp(1.0, Noise, max(uvwa1.a, uvwa2.a));\r\n\tOutExtinction += ConstantExtinction;\r\n    \t\r\n    OutScattering = OutExtinction * Albedo;\r\n\r\n    OutLuminance = SunColor * VolumeSample.g;\r\n\tOutLuminance += FillColor * VolumeSample.a;\r\n\t\r\n\tfloat remap = TemperatureRange > 1e-8 ? saturate((VolumeSample.b - TemperatureOffset) / TemperatureRange) : 0;\r\n\tfloat3 TemperatureCurve = TemperatureRamp.SampleLevel(TemperatureRampSampler, remap, 0 ).rgb;\r\n\t\r\n\tOutLuminance += TemperatureCurve;\r\n    }\r\n\r\n};\r\n\r\nFireRampsRenderVolume Volume;\r\nVolume.VolumeTexture = VolumeTexture;\r\nVolume.VolumeTextureSampler = VolumeTextureSampler;\r\nVolume.Albedo = Albedo;\r\nVolume.SunColor = SunColor;\r\nVolume.FillColor = FillColor;\r\nVolume.ConstantExtinction = ConstantExtinction.r;\r\nVolume.DensityOffset = DensityOffset;\r\nVolume.DensityRange = DensityRange;\r\nVolume.DensityRamp = DensityRamp;\r\nVolume.DensityRampSampler = DensityRampSampler;\r\nVolume.DensityGain = DensityGain;\r\nVolume.TemperatureOffset = TemperatureOffset;\r\nVolume.TemperatureRange = TemperatureRange;\r\nVolume.TemperatureRamp = TemperatureRamp;\r\nVolume.TemperatureRampSampler = TemperatureRampSampler;\r\nVolume.NoiseTexture = NoiseTexture;\r\nVolume.NoiseTextureSampler = NoiseTextureSampler;\r\nVolume.TexCoord1 = TexCoord1;\r\nVolume.TexCoord1Sampler = TexCoord1Sampler;\r\nVolume.TexCoord2 = TexCoord2;\r\nVolume.TexCoord2Sampler = TexCoord2Sampler;\r\nVolume.TexBlend = TexBlend;\r\nVolume.TexScale = TexScale;\r\nVolume.TexFit_0 = TexFit_0;\r\nVolume.TexFit_1 = TexFit_1;\r\n\r\nVolume.TextureScale = 1.0/TexScale;\r\nVolume.FitMin   = TexFit_0;\r\nVolume.FitRange = TexFit_1 - TexFit_0;\r\n\r\nfloat4 ret;\r\nMARCH_SINGLE_RAY(RayStart, RayDir, UnitStepSize, LocalStepSize, NumSteps, Volume, ret);\r\n\r\nreturn ret;"
   description=Raymarch Volume
   outputType=CMOT_Float4
   includeFilePaths [/Plugin/FX/NiagaraFluids/NiagaraFluids.ush]
   inputs [
    {
     inputName=VolumeTexture
     input {expression=&self:MaterialExpressionTextureObjectParameter_2}
    }
    {inputName=RayStart, input {expression=&self:MaterialExpressionMaterialFunctionCall_2}}
    {
     inputName=RayDir
     input {expression=&self:MaterialExpressionMaterialFunctionCall_2, outputIndex=1}
    }
    {
     inputName=NumSteps
     input {expression=&self:MaterialExpressionMaterialFunctionCall_2, outputIndex=2}
    }
    {
     inputName=UnitStepSize
     input {expression=&self:MaterialExpressionMaterialFunctionCall_2, outputIndex=3}
    }
    {
     inputName=LocalStepSize
     input {expression=&self:MaterialExpressionMaterialFunctionCall_2, outputIndex=4}
    }
    {inputName=SunColor, input {expression=&self:MaterialExpressionMultiply_1}}
    {inputName=FillColor, input {expression=&self:MaterialExpressionMultiply_5}}
    {
     inputName=ConstantExtinction
     input {expression=&self:MaterialExpressionScalarParameter_6}
    }
    {inputName=DensityOffset, input {expression=&self:MaterialExpressionScalarParameter_17}}
    {inputName=DensityRange, input {expression=&self:MaterialExpressionScalarParameter_18}}
    {
     inputName=DensityRamp
     input {expression=&self:MaterialExpressionTextureObjectParameter_4}
    }
    {inputName=DensityGain, input {expression=&self:MaterialExpressionScalarParameter_19}}
    {
     inputName=TemperatureOffset
     input {expression=&self:MaterialExpressionScalarParameter_15}
    }
    {
     inputName=TemperatureRange
     input {expression=&self:MaterialExpressionScalarParameter_16}
    }
    {
     inputName=TemperatureRamp
     input {expression=&self:MaterialExpressionTextureObjectParameter_1}
    }
    {inputName=Albedo, input {expression=&self:MaterialExpressionScalarParameter_5}}
    {
     inputName=NoiseTexture
     input {expression=&self:MaterialExpressionTextureObjectParameter_6}
    }
    {
     inputName=TexCoord1
     input {expression=&self:MaterialExpressionTextureObjectParameter_3}
    }
    {
     inputName=TexCoord2
     input {expression=&self:MaterialExpressionTextureObjectParameter_5}
    }
    {inputName=TexBlend, input {expression=&self:MaterialExpressionScalarParameter_8}}
    {inputName=TexScale, input {expression=&self:MaterialExpressionScalarParameter_12}}
    {inputName=TexFit_0, input {expression=&self:MaterialExpressionScalarParameter_13}}
    {inputName=TexFit_1, input {expression=&self:MaterialExpressionScalarParameter_21}}
   ]
  }
  MaterialExpressionTextureObjectParameter_7 = TextureObjectParameter() {parameterName=AlbedoRamp}
  MaterialExpressionTextureObjectParameter_8 = TextureObjectParameter() {parameterName=AlbedoRampParam, samplerType=SAMPLERTYPE_LinearColor, texture=&VolumeToRender}
  n4 = Custom!(
   VolumeTexture=@MaterialExpressionTextureObjectParameter_2
   RayStart=@MaterialExpressionMaterialFunctionCall_2.Ray Origin
   RayDir=@MaterialExpressionMaterialFunctionCall_2.Ray Origin
   NumSteps=@MaterialExpressionMaterialFunctionCall_2.Ray Origin
   UnitStepSize=@MaterialExpressionMaterialFunctionCall_2.Ray Origin
   LocalStepSize=@MaterialExpressionMaterialFunctionCall_2.Ray Origin
   SunColor=@MaterialExpressionMultiply_1
   FillColor=@MaterialExpressionMultiply_5
   ConstantExtinction=@MaterialExpressionScalarParameter_6
   DensityOffset=@MaterialExpressionScalarParameter_17
   DensityRange=@MaterialExpressionScalarParameter_18
   DensityRamp=@MaterialExpressionTextureObjectParameter_4
   DensityGain=@MaterialExpressionScalarParameter_19
   TemperatureOffset=@MaterialExpressionScalarParameter_15
   TemperatureRange=@MaterialExpressionScalarParameter_16
   TemperatureRamp=@MaterialExpressionTextureObjectParameter_1
   Albedo=@MaterialExpressionScalarParameter_5
   AlbedoRamp=@MaterialExpressionTextureObjectParameter_7
   AlbedoRampParam=@MaterialExpressionTextureObjectParameter_8
  ) {
   code="struct FireRampsRenderVolume\r\n{\r\n\tTexture3D VolumeTexture;\r\n\tSamplerState VolumeTextureSampler;\r\n\tfloat Albedo;\r\n\r\n\tfloat3 SunColor;\r\n\tfloat3 FillColor;\r\n\tfloat ConstantExtinction;\r\n    float DensityOffset;\r\n    float DensityRange;\r\n    Texture2D DensityRamp;\r\n\tSamplerState DensityRampSampler;\r\n    float DensityGain;\r\n    float TemperatureOffset;\r\n    float TemperatureRange;\r\n    Texture2D TemperatureRamp;\r\n\tSamplerState TemperatureRampSampler;\r\n\t\r\n\tTexture3D AlbedoRampParam;\r\n\tSamplerState AlbedoRampParamSampler;\r\n\tTexture2D AlbedoRamp;\r\n\tSamplerState AlbedoRampSampler;\r\n\r\n\t// Computes volume extinction coefficient and luminance for a given point in the volume\r\n\tvoid ComputeStep(float3 Position, float3 RayDirection, out float OutExtinction, out float OutScattering, out float3 OutLuminance)\r\n\t{  \r\n\r\n    float4 VolumeSample = VolumeTexture.SampleLevel(VolumeTextureSampler, Position, 0);\r\n\t\r\n    float Density = DensityRange > 1e-8 ? saturate((VolumeSample.r - DensityOffset) / DensityRange) : 0.0;\r\n\tfloat DensityCurve = DensityRamp.SampleLevel(DensityRampSampler, Density, 0 ).r;\r\n\r\n    OutExtinction = DensityCurve * DensityGain;\r\n\tOutExtinction += ConstantExtinction * (OutExtinction > 1e-5);\r\n\r\n    OutScattering = OutExtinction * Albedo;\r\n\r\n    OutLuminance = SunColor * VolumeSample.g;\r\n\tOutLuminance += FillColor * VolumeSample.a;\r\n\t\t\r\n\tfloat AlbedoParam = AlbedoRampParam.SampleLevel(AlbedoRampParamSampler, Position, 0).r;\r\n\tfloat3 AlbedoCurve = AlbedoRamp.SampleLevel(AlbedoRampSampler, AlbedoParam, 0).rgb;\r\n\tOutLuminance *= AlbedoCurve;\r\n\t\r\n\tfloat remap = TemperatureRange > 1e-8 ? saturate((VolumeSample.b - TemperatureOffset) / TemperatureRange) : 0;\r\n\tfloat3 TemperatureCurve = TemperatureRamp.SampleLevel(TemperatureRampSampler, remap, 0 ).rgb;\r\n\t\r\n\tOutLuminance += TemperatureCurve;\r\n\r\n\t}\r\n};\r\n\r\nFireRampsRenderVolume Volume;\r\nVolume.VolumeTexture = VolumeTexture;\r\nVolume.VolumeTextureSampler = VolumeTextureSampler;\r\nVolume.Albedo = Albedo;\r\nVolume.SunColor = SunColor;\r\nVolume.FillColor = FillColor;\r\nVolume.ConstantExtinction = ConstantExtinction.r;\r\nVolume.DensityOffset = DensityOffset;\r\nVolume.DensityRange = DensityRange;\r\nVolume.DensityRamp = DensityRamp;\r\nVolume.DensityRampSampler = DensityRampSampler;\r\nVolume.DensityGain = DensityGain;\r\nVolume.TemperatureOffset = TemperatureOffset;\r\nVolume.TemperatureRange = TemperatureRange;\r\nVolume.TemperatureRamp = TemperatureRamp;\r\nVolume.TemperatureRampSampler = TemperatureRampSampler;\r\nVolume.AlbedoRampParam = AlbedoRampParam;\r\nVolume.AlbedoRampParamSampler = AlbedoRampParamSampler;\r\nVolume.AlbedoRamp = AlbedoRamp;\r\nVolume.AlbedoRampSampler = AlbedoRampSampler;\r\n\r\nfloat4 ret;\r\nMARCH_SINGLE_RAY(RayStart, RayDir, UnitStepSize, LocalStepSize, NumSteps, Volume, ret);\r\n\r\nreturn ret;"
   description=Raymarch Volume
   outputType=CMOT_Float4
   includeFilePaths [&ush]
   inputs [
    {
     inputName=VolumeTexture
     input {expression=&self:MaterialExpressionTextureObjectParameter_2}
    }
    {inputName=RayStart, input {expression=&self:MaterialExpressionMaterialFunctionCall_2}}
    {
     inputName=RayDir
     input {expression=&self:MaterialExpressionMaterialFunctionCall_2, outputIndex=1}
    }
    {
     inputName=NumSteps
     input {expression=&self:MaterialExpressionMaterialFunctionCall_2, outputIndex=2}
    }
    {
     inputName=UnitStepSize
     input {expression=&self:MaterialExpressionMaterialFunctionCall_2, outputIndex=3}
    }
    {
     inputName=LocalStepSize
     input {expression=&self:MaterialExpressionMaterialFunctionCall_2, outputIndex=4}
    }
    {inputName=SunColor, input {expression=&self:MaterialExpressionMultiply_1}}
    {inputName=FillColor, input {expression=&self:MaterialExpressionMultiply_5}}
    {
     inputName=ConstantExtinction
     input {expression=&self:MaterialExpressionScalarParameter_6}
    }
    {inputName=DensityOffset, input {expression=&self:MaterialExpressionScalarParameter_17}}
    {inputName=DensityRange, input {expression=&self:MaterialExpressionScalarParameter_18}}
    {
     inputName=DensityRamp
     input {expression=&self:MaterialExpressionTextureObjectParameter_4}
    }
    {inputName=DensityGain, input {expression=&self:MaterialExpressionScalarParameter_19}}
    {
     inputName=TemperatureOffset
     input {expression=&self:MaterialExpressionScalarParameter_15}
    }
    {
     inputName=TemperatureRange
     input {expression=&self:MaterialExpressionScalarParameter_16}
    }
    {
     inputName=TemperatureRamp
     input {expression=&self:MaterialExpressionTextureObjectParameter_1}
    }
    {inputName=Albedo, input {expression=&self:MaterialExpressionScalarParameter_5}}
    {
     inputName=AlbedoRamp
     input {expression=&self:MaterialExpressionTextureObjectParameter_7}
    }
    {
     inputName=AlbedoRampParam
     input {expression=&self:MaterialExpressionTextureObjectParameter_8}
    }
   ]
  }
  n5 = MaterialFunctionCall() {materialFunction=&MakeFloat4}
  n6 = FeatureLevelSwitch(Default=StaticSwitchParameter(True=@n3, False=@n4) {parameterName=Use Texture Advection}, ES2=@n5.Result, ES3_1=@n5.Result)
  n7 = Max(A=ShadingPathSwitch(Default=QualitySwitch(Default=@n6, Low=@n5.Result), Forward=@n5.Result, Mobile=@n5.Result)) {constB=0}
  Reroute(None=ComponentMask(None=@n7) {a=true})
  n8 = ComponentMask(None=@n7) {b=true, g=true, r=true}
  StaticSwitchParameter(True=EyeAdaptationInverse(LightValueInput=@n8, AlphaInput=Constant() {r=1}).EyeAdaptationInverse, False=@n8) {defaultValue=true, parameterName=InvertEyeAdaptation}
 ]
}
