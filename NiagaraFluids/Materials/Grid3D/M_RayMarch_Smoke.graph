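# Unlit alpha-composite material that raymarches a 3D volume texture to render Niagara Fluids grid smoke on sprites or mesh particles.
# A box-intersection function feeds a Custom HLSL march, where red drives extinction and green scales light color; TemperatureMult is unused.
# Light blends squared LightColor with Schlick-phased sky-atmosphere sunlight via UseSunColor.
# Forward, mobile, low-quality and ES paths substitute a MakeFloat4 constant; color gets inverse eye adaptation, with depth test disabled.
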Material /NiagaraFluids/Materials/Grid3D/M_RayMarch_Smoke
graph {
 properties {
  bDisableDepthTest=true
  bUsedWithNiagaraMeshParticles=true
  bUsedWithNiagaraSprites=true
  blendMode=BLEND_AlphaComposite
  pixelDepthOffsetMode=PDOM_Legacy
  shadingModel=MSM_Unlit
  usedShadingModels=Unlit
 }
 nodes [
  MaterialExpressionTextureObjectParameter_1 = TextureObjectParameter() {parameterName=VolumeTex, samplerType=SAMPLERTYPE_LinearColor, texture=/Engine/EngineDebugMaterials/VolumeToRender}
  MaterialExpressionMaterialFunctionCall_2 = MaterialFunctionCall!(
   Trace Distance=ScalarParameter() {defaultValue=1, group=Raymarch Setup, parameterName=TraceDistance}
   Step Size Scale=ScalarParameter() {defaultValue=1, group=Raymarch Setup, parameterName=Step Size Mult}
  ) {materialFunction=/NiagaraFluids/Materials/Functions/Niagara_Grid3D_BoxIntersection}
  MaterialExpressionScalarParameter_11 = ScalarParameter() {defaultValue=1, parameterName=DensityMult}
  MaterialExpressionScalarParameter_14 = ScalarParameter() {parameterName=TemperatureMult}
  n1 = VectorParameter() {parameterName=LightColor, defaultValue=<rgba 1 1 1 1>}
  n2 = MaterialFunctionCall!(Cosine=DotProduct(A=SkyAtmosphereLightDirection(), B=CameraVectorWS()), Anisotropy=ScalarParameter() {parameterName=Anisotropy}) {materialFunction=/Engine/Functions/Engine_MaterialFunctions01/Volumetrics/SchlickPhase}
  n3 = SkyAtmosphereLightIlluminance!(World Position=Add(A=ObjectPositionWS(), B=VectorParameter() {parameterName=Offset for SkyAtmosphereLight, defaultValue=<rgba 0 0 512 0>}.RGB))
  MaterialExpressionLinearInterpolate_0 = LinearInterpolate(
   A=Multiply(A=@n1.RGB, B=@n1.RGB)
   B=Multiply(A=Multiply(A=@n2.Result, B=@n3), B=ScalarParameter() {defaultValue=2, parameterName=Sun Boost})
   Alpha=ScalarParameter() {parameterName=UseSunColor}
  )
  MaterialExpressionScalarParameter_7 = ScalarParameter() {defaultValue=0.1, parameterName=Albedo}
  n4 = Custom!(
   VolumeTexture=@MaterialExpressionTextureObjectParameter_1
   RayStart=@MaterialExpressionMaterialFunctionCall_2.Ray Origin
   RayDir=@MaterialExpressionMaterialFunctionCall_2.Ray Origin
   NumSteps=@MaterialExpressionMaterialFunctionCall_2.Ray Origin
   UnitStepSize=@MaterialExpressionMaterialFunctionCall_2.Ray Origin
   LocalStepSize=@MaterialExpressionMaterialFunctionCall_2.Ray Origin
   DensityMult=@MaterialExpressionScalarParameter_11
   TemperatureMult=@MaterialExpressionScalarParameter_14
   LightColor=@MaterialExpressionLinearInterpolate_0
   Albedo=@MaterialExpressionScalarParameter_7
  ) {
   code="struct FireRenderVolume\r\n{\r\n\tfloat DensityMult;\r\n\tfloat TemperatureMult;\r\n\tfloat3 LightColor;\r\n\tTexture3D VolumeTexture;\r\n\tSamplerState VolumeTextureSampler;\r\n\tfloat Albedo;\r\n\r\n\t// Computes volume extinction coefficient and luminance for a given point in the volume\r\n\tvoid ComputeStep(float3 Position, float3 RayDirection, out float OutExtinction, out float OutScattering, out float3 OutLuminance)\r\n\t{  \r\n\t\tfloat4 VolumeSample = VolumeTexture.SampleLevel(VolumeTextureSampler, Position, 0);\r\n\r\n\t\tOutExtinction = VolumeSample.r * DensityMult;\r\n\t\tOutScattering = OutExtinction * Albedo;\r\n\r\n\t\tOutLuminance = LightColor * VolumeSample.g;\r\n\t}\r\n};\r\n\r\nFireRenderVolume Volume;\r\nVolume.DensityMult = DensityMult;\r\nVolume.TemperatureMult = TemperatureMult;\r\nVolume.LightColor = LightColor;\r\nVolume.VolumeTexture = VolumeTexture;\r\nVolume.VolumeTextureSampler = VolumeTextureSampler;\r\nVolume.Albedo = Albedo;\r\n\r\nfloat4 ret;\r\nMARCH_SINGLE_RAY(RayStart, RayDir, UnitStepSize, LocalStepSize, NumSteps, Volume, ret);\r\n\r\nreturn ret;"
   description=Raymarch Volume
   outputType=CMOT_Float4
   includeFilePaths [/Plugin/FX/NiagaraFluids/NiagaraFluids.ush]
   inputs [
    {inputName=VolumeTexture, input {expression=&self:MaterialExpressionTextureObjectParameter_1}}
    {inputName=RayStart, input {expression=&self:MaterialExpressionMaterialFunctionCall_2}}
    {
     inputName=RayDir
     input {expression=&self:MaterialExpressionMaterialFunctionCall_2, outputIndex=1}
    }
    {
     inputName=NumSteps
     input {expression=&self:MaterialExpressionMaterialFunctionCall_2, outputIndex=2}
    }
    {
     inputName=UnitStepSize
     input {expression=&self:MaterialExpressionMaterialFunctionCall_2, outputIndex=3}
    }
    {
     inputName=LocalStepSize
     input {expression=&self:MaterialExpressionMaterialFunctionCall_2, outputIndex=4}
    }
    {inputName=DensityMult, input {expression=&self:MaterialExpressionScalarParameter_11}}
    {inputName=TemperatureMult, input {expression=&self:MaterialExpressionScalarParameter_14}}
    {inputName=LightColor, input {expression=&self:MaterialExpressionLinearInterpolate_0}}
    {inputName=Albedo, input {expression=&self:MaterialExpressionScalarParameter_7}}
   ]
  }
  n5 = MaterialFunctionCall() {materialFunction=&MakeFloat4}
  n6 = ShadingPathSwitch(
   Default=QualitySwitch(Default=FeatureLevelSwitch(Default=@n4, ES2=@n5.Result, ES3_1=@n5.Result), Low=@n5.Result)
   Forward=@n5.Result
   Mobile=@n5.Result
  )
  ComponentMask(None=@n6) {a=true}
  EyeAdaptationInverse(LightValueInput=ComponentMask(None=@n6) {b=true, g=true, r=true}, AlphaInput=Constant() {r=1})
 ]
}
