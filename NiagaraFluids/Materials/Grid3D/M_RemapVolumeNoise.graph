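# Volume-domain material for Niagara Fluids 3D grids that renders a noise-detailed density field as a heterogeneous volume.
# Transformed world position, scaled by WorldGridExtents and shifted by LocalPivot, samples VolumeTex and AttrTex.
# VolumeTex RGB serves as the noise lookup position, so detail follows stored coordinates; the powered noise is masked by its alpha and multiplied by AttrTex alpha.
# The result is range-remapped, optionally shaped by a DensityRamp texture, then gained, with ConstantExtinction added where positive.
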Material /NiagaraFluids/Materials/Grid3D/M_RemapVolumeNoise
graph {
 properties {
  bUsedWithHeterogeneousVolumes=true
  bUsedWithNiagaraMeshParticles=true
  blendMode=BLEND_Additive
  materialDomain=MD_Volume
  pixelDepthOffsetMode=PDOM_Legacy
  usedShadingModels=Default Lit
 }
 nodes [
  Constant3Vector() {constant=<rgba 1 0 0 0>}
  n1 = WorldPosition()
  n2 = Add(
   A=Divide(A=TransformPosition(None=@n1.XYZ) {transformSourceType=TRANSFORMPOSSOURCE_World}, B=VectorParameter() {parameterName=WorldGridExtents}.RGB)
   B=VectorParameter() {parameterName=LocalPivot}.RGB
  )
  n3 = TextureSample(UVs=@n2, Tex=TextureObjectParameter() {parameterName=VolumeTex, texture=/Engine/EngineResources/DefaultVolumeTexture}) {constCoordinate=1}
  n4 = Noise!(World Position=@n3.RGB, FilterWidth) {outputMax=0.1, outputMin=0, scale=4}
  n5 = SmoothStep(Value=@n3.A) {constMin=-1}
  n6 = Multiply(
   A=TextureSample(UVs=@n2, Tex=TextureObjectParameter() {parameterName=AttrTex, texture=&DefaultVolumeTexture}) {constCoordinate=1}.A
   B=Multiply(A=Power(Base=@n4) {constExponent=4}, B=OneMinus(None=@n5))
  )
  n7 = ScalarParameter() {parameterName=DensityOffset}
  n8 = ScalarParameter() {parameterName=DensityRange}
  n9 = Constant() {r=1}
  n10 = Saturate(None=Divide(A=Subtract(A=@n6, B=@n7), B=If(A=@n8, "A > B"=@n8, "A == B"=@n9, "A < B"=@n9)))
  n11 = Multiply(
   A=StaticSwitchParameter(
    True=TextureSampleParameter2D(UVs=@n10) {parameterName=DensityRamp, samplerType=SAMPLERTYPE_LinearColor, texture=/Engine/ArtTools/RenderToTexture/Textures/127grey}.R
    False=@n10
   ) {parameterName=Use Density Curve}
   B=ScalarParameter() {defaultValue=1000, parameterName=DensityGain}
  )
  If(A=@n11, "A > B"=Add(A=@n11, B=ScalarParameter() {parameterName=ConstantExtinction}), "A == B"=@n11, "A < B"=@n11) {constB=0}
 ]
}
