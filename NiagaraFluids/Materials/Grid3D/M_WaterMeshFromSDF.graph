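# Masked Single Layer Water material for Niagara Fluids 3D grid water, usable on mesh particles and sprites, named for meshes derived from an SDF.
# Depth test is disabled and normals are world-space, though no SDF or normal logic appears in the graph.
# Scattering and Absorption color parameters are each squared into the water output's coefficients, alongside a Phase parameter.
# Base Color, Metallic, Specular and Opacity parameters are also declared, with no visible connections.
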Material /NiagaraFluids/Materials/Grid3D/M_WaterMeshFromSDF
graph {
 properties {
  bDisableDepthTest=true
  bTangentSpaceNormal=false
  bUsedWithNiagaraMeshParticles=true
  bUsedWithNiagaraSprites=true
  blendMode=BLEND_Masked
  pixelDepthOffsetMode=PDOM_Legacy
  shadingModel=MSM_SingleLayerWater
  usedShadingModels=SingleLayerWater
 }
 nodes [
  Constant()
  ScalarParameter() {defaultValue=1, parameterName=Opacity}
  ScalarParameter() {parameterName=Specular}
  ScalarParameter() {parameterName=Metallic}
  n1 = VectorParameter() {parameterName=Scattering, defaultValue=<rgba 0.0532 0.2396 0.7292 0>}
  n2 = VectorParameter() {parameterName=Absorption, defaultValue=<rgba 0.6 0.05 0.005 0>}
  SingleLayerWaterMaterialOutput!(
   ScatteringCoefficients=Multiply(A=@n1.RGB, B=@n1.RGB)
   AbsorptionCoefficients=Multiply(A=@n2.RGB, B=@n2.RGB)
   PhaseG=ScalarParameter() {parameterName=Phase, sliderMin=0.5}
   ColorScaleBehindWater
  )
  VectorParameter() {parameterName=Base Color}
 ]
}
