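# Masked Single Layer Water material for Niagara Fluids that renders a 3D grid liquid surface by sphere-tracing a signed distance field volume texture.
# Camera rays, perspective or orthographic, move to local space through a legacy matrix or built-in transform, then intersect the grid box.
# Custom HLSL outputs hit mask, sampled or central-difference normal, depth for pixel offset, and whitewater that scales scattering.
# A debug parameter draws the bounds as a voxel checkerboard.
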Material /NiagaraFluids/Materials/Grid3D/M_WaterSDF
graph {
 properties {
  bDisableDepthTest=true
  bTangentSpaceNormal=false
  bUsedWithNiagaraMeshParticles=true
  bUsedWithNiagaraSprites=true
  blendMode=BLEND_Masked
  pixelDepthOffsetMode=PDOM_Legacy
  shadingModel=MSM_SingleLayerWater
  usedShadingModels=SingleLayerWater
 }
 nodes [
  n1 = IsOrthographic()
  n2 = WorldPosition()
  n3 = MaterialFunctionCall() {materialFunction=&CameraDirectionVector}
  n4 = Subtract(A=@n2.XYZ, B=Multiply(A=PixelDepth().R, B=@n3.Result))
  n5 = CameraPositionWS()
  n6 = If(A=@n1, "A > B"=@n4, "A == B"=@n4, "A < B"=@n5) {constB=1, desc=Ray Start Ortho Switch}
  MaterialExpressionReroute_0 = Reroute(None=@n6)
  n7 = If(A=@n1, "A > B"=@n3.Result, "A == B"=@n3.Result, "A < B"=Normalize(VectorInput=Subtract(A=@n2.XYZ, B=@n5))) {constB=1, desc=Ray Dir Ortho Switch}
  MaterialExpressionReroute_10 = Reroute(None=@n7)
  MaterialExpressionReroute_5 = Reroute(None=TextureObjectParameter() {parameterName=VolumeTex, samplerType=SAMPLERTYPE_LinearColor, texture=/Engine/EngineDebugMaterials/VolumeToRender})
  MaterialExpressionReroute_6 = Reroute(None=VectorParameter() {parameterName=WorldGridExtents}.RGB)
  MaterialExpressionReroute_7 = Reroute(None=ScalarParameter() {defaultValue=0.25, parameterName=Voxel Size})
  MaterialExpressionConstant_0 = Constant()
  MaterialExpressionReroute_11 = Reroute(None=@n3.Result)
  MaterialExpressionConstant_3 = Constant() {r=100000}
  MaterialExpressionScalarParameter_4 = ScalarParameter() {parameterName=Debug Bounds}
  MaterialExpressionVectorParameter_1 = VectorParameter() {parameterName=DebugSliceCenter, defaultValue=<rgba 0 0 100 0>}
  MaterialExpressionVectorParameter_2 = VectorParameter() {parameterName=DebugSliceAxis, defaultValue=<rgba 0 1 0 0>}
  MaterialExpressionScalarParameter_1 = ScalarParameter() {defaultValue=1, parameterName=Tolerance}
  MaterialExpressionScalarParameter_2 = ScalarParameter() {defaultValue=300, parameterName=NumSteps}
  MaterialExpressionScalarParameter_3 = ScalarParameter() {defaultValue=1, parameterName=ComputeNormals}
  MaterialExpressionReroute_14 = Reroute(None=@n6)
  MaterialExpressionReroute_15 = Reroute(None=@n7)
  n8 = VectorParameter() {parameterName=WorldToLocal0}
  MaterialExpressionAppendVector_1 = AppendVector(A=@n8.RGB, B=@n8.RGB)
  n9 = VectorParameter() {parameterName=WorldToLocal1}
  MaterialExpressionAppendVector_2 = AppendVector(A=@n9.RGB, B=@n9.RGB)
  n10 = VectorParameter() {parameterName=WorldToLocal2}
  MaterialExpressionAppendVector_3 = AppendVector(A=@n10.RGB, B=@n10.RGB)
  n11 = VectorParameter() {parameterName=WorldToLocal3}
  MaterialExpressionAppendVector_4 = AppendVector(A=@n11.RGB, B=@n11.RGB)
  n12 = Custom!(
   WorldRayStart=@MaterialExpressionReroute_14
   WorldRayDir=@MaterialExpressionReroute_15
   WorldToLocal0=@MaterialExpressionAppendVector_1
   WorldToLocal1=@MaterialExpressionAppendVector_2
   WorldToLocal2=@MaterialExpressionAppendVector_3
   WorldToLocal3=@MaterialExpressionAppendVector_4
  ) {
   code="float4x4 WorldToLocal = {\r\nWorldToLocal0.x, WorldToLocal0.y, WorldToLocal0.z, WorldToLocal0.w, \r\nWorldToLocal1.x, WorldToLocal1.y, WorldToLocal1.z, WorldToLocal1.w, \r\nWorldToLocal2.x, WorldToLocal2.y, WorldToLocal2.z, WorldToLocal2.w, \r\nWorldToLocal3.x, WorldToLocal3.y, WorldToLocal3.z, WorldToLocal3.w\r\n};\r\n\r\n\r\nLocalRayStart = mul(float4(WorldRayStart,1), WorldToLocal).xyz;\r\nLocalRayDir = mul(float4(WorldRayDir,0), WorldToLocal).xyz;\r\nreturn 0;"
   outputType=CMOT_Float1
   showCode=true
   additionalOutputs [{outputName=LocalRayStart, outputType=CMOT_Float3}, {outputName=LocalRayDir, outputType=CMOT_Float3}]
   inputs [
    {inputName=WorldRayStart, input {expression=&self:MaterialExpressionReroute_14}}
    {inputName=WorldRayDir, input {expression=&self:MaterialExpressionReroute_15}}
    {inputName=WorldToLocal0, input {expression=&self:MaterialExpressionAppendVector_1}}
    {inputName=WorldToLocal1, input {expression=&self:MaterialExpressionAppendVector_2}}
    {inputName=WorldToLocal2, input {expression=&self:MaterialExpressionAppendVector_3}}
    {inputName=WorldToLocal3, input {expression=&self:MaterialExpressionAppendVector_4}}
   ]
  }
  n13 = TransformPosition(None=@MaterialExpressionReroute_14) {transformSourceType=TRANSFORMPOSSOURCE_World}
  n14 = Multiply(A=VectorParameter() {parameterName=LocalPivot}.RGB, B=VectorParameter() {parameterName=WorldGridExtents}.RGB)
  MaterialExpressionReroute_13 = Reroute(None=StaticSwitchParameter(True=@n12.LocalRayStart, False=Subtract(A=@n13, B=@n14)) {defaultValue=true, parameterName=Use Legacy Transform})
  MaterialExpressionReroute_12 = Reroute(None=StaticSwitchParameter(
    True=@n12.LocalRayDir
    False=Transform(None=@MaterialExpressionReroute_15) {transformSourceType=TRANSFORMSOURCE_World, transformType=TRANSFORM_Local}
   ) {defaultValue=true, parameterName=Use Legacy Transform})
  n15 = Custom!(
   WorldRayStart=@MaterialExpressionReroute_0
   WorldRayDir=@MaterialExpressionReroute_10
   VolumeTexture=@MaterialExpressionReroute_5
   WorldGridExtents=@MaterialExpressionReroute_6
   VoxelSize=@MaterialExpressionReroute_7
   WorldToLocal0=@MaterialExpressionConstant_0
   WorldToLocal1=@MaterialExpressionConstant_0
   WorldToLocal2=@MaterialExpressionConstant_0
   WorldToLocal3=@MaterialExpressionConstant_0
   CameraDirectionVector=@MaterialExpressionReroute_11
   SceneDepth=@MaterialExpressionConstant_3
   DebugBBox=@MaterialExpressionScalarParameter_4
   Center=@MaterialExpressionVectorParameter_1.RGB
   Normal=@MaterialExpressionVectorParameter_2.RGB
   Tolerance=@MaterialExpressionScalarParameter_1
   NumSteps=@MaterialExpressionScalarParameter_2
   ComputeNormals=@MaterialExpressionScalarParameter_3
   LocalRayStart=@MaterialExpressionReroute_13
   LocalRayDir=@MaterialExpressionReroute_12
  ) {
   code="float4 ret = {0,0,0,0};\r\n\r\nfloat WorldDepth = SceneDepth / dot(LocalRayDir, CameraDirectionVector);\r\n\r\n//\r\n//  Ray aabbox intersection\r\n//\r\n\r\nfloat3 dirfrac = 1.0f / LocalRayDir;\r\n\r\nfloat t = 0;\r\n\r\n\r\nfloat3 MinBounds = -1.* WorldGridExtents * .5;\r\nfloat3 MaxBounds = WorldGridExtents * .5;\r\nfloat t1 = (MinBounds.x - LocalRayStart.x)*dirfrac.x;\r\nfloat t2 = (MaxBounds.x - LocalRayStart.x)*dirfrac.x;\r\nfloat t3 = (MinBounds.y - LocalRayStart.y)*dirfrac.y;\r\nfloat t4 = (MaxBounds.y - LocalRayStart.y)*dirfrac.y;\r\nfloat t5 = (MinBounds.z - LocalRayStart.z)*dirfrac.z;\r\nfloat t6 = (MaxBounds.z - LocalRayStart.z)*dirfrac.z;\r\n\r\nfloat tmin = max(max(min(t1, t2), min(t3, t4)), min(t5, t6));\r\nfloat tmax = min(min(max(t1, t2), max(t3, t4)), max(t5, t6));\r\n\r\n// if tmax < 0, ray (line) is intersecting AABB, but the whole AABB is behind us\r\nif (tmax < 0)\r\n{\r\n    t = tmax;\r\n    return ret;\r\n}\r\n\r\n// if tmin > tmax, ray doesn't intersect AABB\r\nif (tmin > tmax)\r\n{\r\n    t = tmax;\r\n    return ret;\r\n}\r\n\r\nt = max(0, tmin);\r\n\r\nfloat3 CurrLocalPos = LocalRayStart + LocalRayDir * t;\r\n\r\nif (DebugBBox > 1e-3 &&&& t < WorldDepth) {\r\n  float4 rettmp = {1,1,1,DebugBBox};\r\n\r\n  int3 VoxelCount = floor(CurrLocalPos / VoxelSize);\r\n\r\n  int SliceParity = (VoxelCount.x) % 2;\r\n  int RowParity = (VoxelCount.y + SliceParity+1) % 2;\r\n  int CellParity = (VoxelCount.z + RowParity ) % 2;\r\n\r\n  if (CellParity == 0)\r\n  {\r\n    rettmp.y = 0;\r\n  }\r\n\r\n  return rettmp;\r\n}\r\n\r\nfloat SignedDistance = -10000;\r\nfloat3 FinalPos = float3(0,0,0);\r\nfloat4 RetVal = float4(0,0,0,0);\r\nWhitewater = 0;\r\nfor (int i = 0; i < NumSteps; ++i)\r\n{\r\n   \r\n    float3 Position = CurrLocalPos / WorldGridExtents + .5;\r\n    float4 VolumeSample = VolumeTexture.SampleLevel(VolumeTextureSampler, Position, 0);\r\n    SignedDistance = VolumeSample.r;\r\n    if (abs(SignedDistance) < VoxelSize * Tolerance &&&& Position.x >= 0 &&&& Position.x <= 1 &&&& Position.y >= 0 &&&& Position.y <= 1 &&&& Position.z >= 0 &&&& Position.z <= 1)\r\n    {\r\n        FinalPos = Position;\r\n        RetVal = float4(normalize(VolumeSample.gba), 1);\r\n        Whitewater = VolumeSample.g;\r\n        Depth = t*dot(WorldRayDir, CameraDirectionVector);\r\n        break;\r\n    }\r\n    \r\n    t += SignedDistance;\r\n    CurrLocalPos = LocalRayStart + LocalRayDir * t;\r\n}\r\n\r\n\r\nif (t > WorldDepth)\r\n{\r\nRetVal.a = 0;\r\n}\r\n\r\nif (RetVal.a != 0 &&&& ComputeNormals > 1-1e-5)\r\n{\r\nuint sx,sy,sz,l;\r\nVolumeTexture.GetDimensions(0, sx, sy, sz, l);\r\nfloat3 UnitDx = 1./float3(sx,sy,sz);\r\n\r\nfloat S_right = VolumeTexture.SampleLevel(VolumeTextureSampler, FinalPos + float3(1,0,0) * UnitDx, 0).r;\r\nfloat S_left = VolumeTexture.SampleLevel(VolumeTextureSampler, FinalPos - \r\nfloat3(1,0,0) * UnitDx, 0).r;\r\nfloat S_up  = VolumeTexture.SampleLevel(VolumeTextureSampler, FinalPos + \r\nfloat3(0,1,0) * UnitDx, 0).r;\r\nfloat S_down = VolumeTexture.SampleLevel(VolumeTextureSampler, FinalPos - float3(0,1,0) * UnitDx, 0).r;\r\nfloat S_front = VolumeTexture.SampleLevel(VolumeTextureSampler, FinalPos + float3(0,0,1) * UnitDx, 0).r;\r\nfloat S_back = VolumeTexture.SampleLevel(VolumeTextureSampler, FinalPos - float3(0,0,1) * UnitDx, 0).r;\r\n\r\nRetVal.rgb = normalize(float3(S_right - S_left, S_up - S_down, S_front - S_back));\r\n\r\n}\r\n\r\nreturn RetVal;"
   outputType=CMOT_Float4
   additionalOutputs [{outputName=Depth, outputType=CMOT_Float1}, {outputName=Whitewater, outputType=CMOT_Float1}]
   inputs [
    {inputName=WorldRayStart, input {expression=&self:MaterialExpressionReroute_0}}
    {inputName=WorldRayDir, input {expression=&self:MaterialExpressionReroute_10}}
    {inputName=VolumeTexture, input {expression=&self:MaterialExpressionReroute_5}}
    {inputName=WorldGridExtents, input {expression=&self:MaterialExpressionReroute_6}}
    {inputName=VoxelSize, input {expression=&self:MaterialExpressionReroute_7}}
    {inputName=WorldToLocal0, input {expression=&self:MaterialExpressionConstant_0}}
    {inputName=WorldToLocal1, input {expression=&self:MaterialExpressionConstant_0}}
    {inputName=WorldToLocal2, input {expression=&self:MaterialExpressionConstant_0}}
    {inputName=WorldToLocal3, input {expression=&self:MaterialExpressionConstant_0}}
    {inputName=CameraDirectionVector, input {expression=&self:MaterialExpressionReroute_11}}
    {inputName=SceneDepth, input {expression=&self:MaterialExpressionConstant_3}}
    {inputName=DebugBBox, input {expression=&self:MaterialExpressionScalarParameter_4}}
    {
     inputName=Center
     input {expression=&self:MaterialExpressionVectorParameter_1, mask=1, maskB=1, maskG=1, maskR=1}
    }
    {
     inputName=Normal
     input {expression=&self:MaterialExpressionVectorParameter_2, mask=1, maskB=1, maskG=1, maskR=1}
    }
    {inputName=Tolerance, input {expression=&self:MaterialExpressionScalarParameter_1}}
    {inputName=NumSteps, input {expression=&self:MaterialExpressionScalarParameter_2}}
    {inputName=ComputeNormals, input {expression=&self:MaterialExpressionScalarParameter_3}}
    {inputName=LocalRayStart, input {expression=&self:MaterialExpressionReroute_13}}
    {inputName=LocalRayDir, input {expression=&self:MaterialExpressionReroute_12}}
   ]
  }
  ComponentMask(None=@n15.return) {b=true, g=true, r=true}
  ComponentMask(None=@n15.return) {a=true}
  Constant()
  MaterialFunctionCall() {materialFunction=/NiagaraFluids/Materials/Functions/Niagara_Liquid_Generate_WPO}
  ScalarParameter() {defaultValue=1, parameterName=Opacity}
  ScalarParameter() {parameterName=Specular}
  ScalarParameter() {parameterName=Metallic}
  n16 = VectorParameter() {parameterName=Scattering, defaultValue=<rgba 0.0532 0.2396 0.7292 0>}
  n17 = ScalarParameter() {defaultValue=1, parameterName=Whitewater}
  n18 = VectorParameter() {parameterName=Absorption, defaultValue=<rgba 0.6 0.05 0.005 0>}
  SingleLayerWaterMaterialOutput!(
   ScatteringCoefficients=Multiply(A=Multiply(A=@n16.RGB, B=@n16.RGB), B=Multiply(A=@n15.Whitewater, B=@n17) {constB=2})
   AbsorptionCoefficients=Multiply(A=@n18.RGB, B=@n18.RGB)
   PhaseG=ScalarParameter() {parameterName=Phase, sliderMin=0.5}
   ColorScaleBehindWater
  )
  Subtract(A=@n15.Depth, B=PixelDepth().R)
  VectorParameter() {parameterName=Base Color}
 ]
}
