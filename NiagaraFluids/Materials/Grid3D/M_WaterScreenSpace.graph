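# Masked Single Layer Water material for Niagara Fluids that renders a screen-space water surface on Niagara sprites or mesh particles, with depth test disabled.
# Samples the Niagara-supplied NormalDepthTex at scaled UVs, deriving an edge cutout from alpha depth and a depth difference against PixelDepth.
# Squares the Absorption and Scattering colors for the water output, with scattering also scaled by Whitewater and a zero-default constant, which cancels it.
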Material /NiagaraFluids/Materials/Grid3D/M_WaterScreenSpace
graph {
 properties {
  bDisableDepthTest=true
  bTangentSpaceNormal=false
  bUsedWithNiagaraMeshParticles=true
  bUsedWithNiagaraSprites=true
  blendMode=BLEND_Masked
  pixelDepthOffsetMode=PDOM_Legacy
  shadingModel=MSM_SingleLayerWater
  usedShadingModels=SingleLayerWater
 }
 nodes [
  Constant()
  n1 = VectorParameter() {parameterName=UVScale, defaultValue=<rgba 1 1 1 0>}
  n2 = TextureSample(
   UVs=Multiply(A=TextureCoordinate(), B=ComponentMask(None=@n1.RGB) {g=true, r=true})
   Tex=TextureObjectParameter() {group=FromNiagara, parameterName=NormalDepthTex}
  )
  n3 = Subtract(A=Divide(A=@n2.A) {constB=40960}, B=ScalarParameter() {group=Opacity, parameterName=FluidEdgeOffset})
  OneMinus(None=Floor(None=Saturate(None=@n3)))
  ScalarParameter() {defaultValue=1, parameterName=Opacity}
  ScalarParameter() {parameterName=Specular}
  ScalarParameter() {parameterName=Metallic}
  n4 = VectorParameter() {parameterName=Scattering, defaultValue=<rgba 0.0532 0.2396 0.7292 0>}
  n5 = Constant()
  n6 = ScalarParameter() {defaultValue=1, parameterName=Whitewater}
  n7 = VectorParameter() {parameterName=Absorption, defaultValue=<rgba 0.6 0.05 0.005 0>}
  SingleLayerWaterMaterialOutput!(
   ScatteringCoefficients=Multiply(A=Multiply(A=@n4.RGB, B=@n4.RGB), B=Multiply(A=@n5, B=@n6) {constB=2})
   AbsorptionCoefficients=Multiply(A=@n7.RGB, B=@n7.RGB)
   PhaseG=ScalarParameter() {parameterName=Phase, sliderMin=0.5}
   ColorScaleBehindWater
  )
  Subtract(A=@n2.A, B=PixelDepth().R)
  VectorParameter() {parameterName=Base Color}
 ]
}
